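# Control Rig for the SpacePlunge Hunt male medium cosmetic body, driving its deform bones and floating kinematic accessory bones.
# After the shared DeformRig function, weighted parent constraints attach arm-box bones to arm bones, torso bones to spine/clavicles (front and back ignoring rotation), and the collar to the neck.
# Transform offsets adjust some bones, then a timeline sine/cosine function bobs torso pieces with per-bone phase offsets.
# A helper converts swing/twist rotation into remapped alphas.
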ControlRigBlueprint /BRCosmetics/Characters/Player/Male/Medium/Bodies/M_MED_SpacePlunge_Hunt/Meshes/M_MED_SpacePlunge_Hunt_CtrlRig
graphs [
 CollapseNode_ContainedGraph {nodes [
   n1 = RigVMFunctionEntryNode()
   RigVMFunction_Sequence = RigVMAggregateNode(ExecuteContext=@n1.ExecutePin)
   &RigUnit_ParentConstraint(
    AdvancedSettings=(RotationOrderForFilter=XZY)
    Child=(Type=Bone,Name="kin_armBoxes_l_1")
    Parents="((Item=(Type=Bone,Name=\"upperarm_twist_01_l\"),Weight=0.0012),(Item=(Type=Bone,Name=\"deform_elbow_out_l\"),Weight=0.0098),(Item=(Type=Bone,Name=\"upperarm_twist_02_l\"),Weight=0.0163),(Item=(Type=Bone,Name=\"lowerarm_twist_01_l\"),Weight=0.071),(Item=(Type=Bone, ...(truncated)"
    Weight=1
    ExecutePin=@RigVMFunction_Sequence.A
   )
   > &RigUnit_ParentConstraint(
    AdvancedSettings=(RotationOrderForFilter=XZY)
    Child=(Type=Bone,Name="kin_armBoxes_l_2")
    Parents="((Item=(Type=Bone,Name=\"deform_wrist_up_l\"),Weight=0.0002),(Item=(Type=Bone,Name=\"upperarm_twist_02_l\"),Weight=0.0132),(Item=(Type=Bone,Name=\"deform_elbow_in_l\"),Weight=0.0361),(Item=(Type=Bone,Name=\"lowerarm_twist_01_l\"),Weight=0.0648),(Item=(Type=Bone,Na ...(truncated)"
    Weight=1
   )
   > &RigUnit_ParentConstraint(
    AdvancedSettings=(RotationOrderForFilter=XZY)
    Child=(Type=Bone,Name="kin_armBoxes_l_3")
    Parents="((Item=(Type=Bone,Name=\"upperarm_twist_01_l\"),Weight=0.0003),(Item=(Type=Bone,Name=\"deform_elbow_out_l\"),Weight=0.0004),(Item=(Type=Bone,Name=\"upperarm_twist_02_l\"),Weight=0.0053),(Item=(Type=Bone,Name=\"lowerarm_twist_01_l\"),Weight=0.0635),(Item=(Type=Bone ...(truncated)"
    Weight=1
   )
   > &RigUnit_ParentConstraint(
    AdvancedSettings=(RotationOrderForFilter=XZY)
    Child=(Type=Bone,Name="kin_armBoxes_l_4")
    Parents="((Item=(Type=Bone,Name=\"deform_elbow_in_l\"),Weight=0.0023),(Item=(Type=Bone,Name=\"upperarm_twist_02_l\"),Weight=0.0024),(Item=(Type=Bone,Name=\"upperarm_l\"),Weight=0.0753),(Item=(Type=Bone,Name=\"lowerarm_twist_01_l\"),Weight=0.0894),(Item=(Type=Bone,Name=\"def ...(truncated)"
    Weight=1
   )
   > &RigUnit_ParentConstraint(
    AdvancedSettings=(RotationOrderForFilter=XZY)
    Child=(Type=Bone,Name="kin_armBoxes_l_5")
    Parents="((Item=(Type=Bone,Name=\"upperarm_twist_01_l\"),Weight=0.0005),(Item=(Type=Bone,Name=\"upperarm_twist_02_l\"),Weight=0.0158),(Item=(Type=Bone,Name=\"deform_elbow_out_l\"),Weight=0.033),(Item=(Type=Bone,Name=\"lowerarm_twist_01_l\"),Weight=0.0399),(Item=(Type=Bone, ...(truncated)"
    Weight=1
   )
   > &RigUnit_ParentConstraint(
    AdvancedSettings=(RotationOrderForFilter=XZY)
    Child=(Type=Bone,Name="kin_armBoxes_l_6")
    Parents="((Item=(Type=Bone,Name=\"deform_upperarm_bk_l\"),Weight=0.0017),(Item=(Type=Bone,Name=\"deform_elbow_in_l\"),Weight=0.0104),(Item=(Type=Bone,Name=\"lowerarm_twist_02_l\"),Weight=0.019),(Item=(Type=Bone,Name=\"upperarm_twist_01_l\"),Weight=0.0562),(Item=(Type=Bone, ...(truncated)"
    Weight=1
   )
   > &RigUnit_ParentConstraint(
    AdvancedSettings=(RotationOrderForFilter=XZY)
    Child=(Type=Bone,Name="kin_armBoxes_l_7")
    Parents="((Item=(Type=Bone,Name=\"upperarm_twist_02_l\"),Weight=0.0052),(Item=(Type=Bone,Name=\"deform_elbow_in_l\"),Weight=0.0357),(Item=(Type=Bone,Name=\"lowerarm_twist_01_l\"),Weight=0.0682),(Item=(Type=Bone,Name=\"deform_elbow_out_l\"),Weight=0.0838),(Item=(Type=Bone,N ...(truncated)"
    Weight=1
   )
   > &RigUnit_ParentConstraint(
    AdvancedSettings=(RotationOrderForFilter=XZY)
    Child=(Type=Bone,Name="kin_armBoxes_l_8")
    Parents="((Item=(Type=Bone,Name=\"deform_elbow_out_l\"),Weight=0.0007),(Item=(Type=Bone,Name=\"deform_upperarm_fr_l\"),Weight=0.0007),(Item=(Type=Bone,Name=\"lowerarm_twist_02_l\"),Weight=0.0403),(Item=(Type=Bone,Name=\"upperarm_twist_01_l\"),Weight=0.0835),(Item=(Type=Bon ...(truncated)"
    Weight=1
   )
   > &RigUnit_ParentConstraint(
    AdvancedSettings=(RotationOrderForFilter=XZY)
    Child=(Type=Bone,Name="kin_armBoxes_l_9")
    Parents="((Item=(Type=Bone,Name=\"deform_elbow_out_l\"),Weight=0.0007),(Item=(Type=Bone,Name=\"deform_upperarm_fr_l\"),Weight=0.0039),(Item=(Type=Bone,Name=\"lowerarm_twist_02_l\"),Weight=0.022),(Item=(Type=Bone,Name=\"lowerarm_l\"),Weight=0.0653),(Item=(Type=Bone,Name=\"de ...(truncated)"
    Weight=1
   )
   > &RigUnit_ParentConstraint(
    AdvancedSettings=(RotationOrderForFilter=XZY)
    Child=(Type=Bone,Name="kin_armBoxes_l_10")
    Parents="((Item=(Type=Bone,Name=\"deform_upperarm_bk_l\"),Weight=0.0011),(Item=(Type=Bone,Name=\"deform_upperarm_up_l\"),Weight=0.0014),(Item=(Type=Bone,Name=\"lowerarm_l\"),Weight=0.0087),(Item=(Type=Bone,Name=\"deform_elbow_in_l\"),Weight=0.013),(Item=(Type=Bone,Name=\"de ...(truncated)"
    Weight=1
   )
   > &RigUnit_ParentConstraint(
    AdvancedSettings=(RotationOrderForFilter=XZY)
    Child=(Type=Bone,Name="kin_armBoxes_l_11")
    Parents="((Item=(Type=Bone,Name=\"deform_upperarm_fr_l\"),Weight=0.0011),(Item=(Type=Bone,Name=\"deform_upperarm_bk_l\"),Weight=0.0451),(Item=(Type=Bone,Name=\"upperarm_l\"),Weight=0.0726),(Item=(Type=Bone,Name=\"upperarm_twist_02_l\"),Weight=0.1661),(Item=(Type=Bone,Name= ...(truncated)"
    Weight=1
   )
   > &RigUnit_ParentConstraint(
    AdvancedSettings=(RotationOrderForFilter=XZY)
    Child=(Type=Bone,Name="kin_armBoxes_l_12")
    Parents="((Item=(Type=Bone,Name=\"deform_elbow_in_l\"),Weight=0.0052),(Item=(Type=Bone,Name=\"deform_upperarm_up_l\"),Weight=0.0058),(Item=(Type=Bone,Name=\"deform_upperarm_bk_l\"),Weight=0.0203),(Item=(Type=Bone,Name=\"lowerarm_l\"),Weight=0.0235),(Item=(Type=Bone,Name=\"d ...(truncated)"
    Weight=1
   )
   > &RigUnit_ParentConstraint(
    AdvancedSettings=(RotationOrderForFilter=XZY)
    Child=(Type=Bone,Name="kin_armBoxes_l_13")
    Parents="((Item=(Type=Bone,Name=\"deform_upperarm_fr_l\"),Weight=0.002),(Item=(Type=Bone,Name=\"deform_elbow_out_l\"),Weight=0.027),(Item=(Type=Bone,Name=\"lowerarm_twist_02_l\"),Weight=0.0301),(Item=(Type=Bone,Name=\"deform_elbow_in_l\"),Weight=0.0713),(Item=(Type=Bone,Na ...(truncated)"
    Weight=1
   )
   > &RigUnit_ParentConstraint(
    AdvancedSettings=(RotationOrderForFilter=XZY)
    Child=(Type=Bone,Name="kin_armBoxes_l_14")
    Parents="((Item=(Type=Bone,Name=\"deform_upperarm_fr_l\"),Weight=0.0003),(Item=(Type=Bone,Name=\"deform_elbow_out_l\"),Weight=0.0124),(Item=(Type=Bone,Name=\"lowerarm_twist_02_l\"),Weight=0.0486),(Item=(Type=Bone,Name=\"upperarm_twist_01_l\"),Weight=0.0602),(Item=(Type=Bon ...(truncated)"
    Weight=1
   )
   > &RigUnit_ParentConstraint(
    AdvancedSettings=(RotationOrderForFilter=XZY)
    Child=(Type=Bone,Name="kin_armBoxes_l_15")
    Parents="((Item=(Type=Bone,Name=\"deform_upperarm_bk_l\"),Weight=0.0022),(Item=(Type=Bone,Name=\"lowerarm_twist_02_l\"),Weight=0.0288),(Item=(Type=Bone,Name=\"deform_elbow_in_l\"),Weight=0.0383),(Item=(Type=Bone,Name=\"upperarm_twist_01_l\"),Weight=0.0822),(Item=(Type=Bone ...(truncated)"
    Weight=1
   )
   > &RigUnit_ParentConstraint(
    AdvancedSettings=(RotationOrderForFilter=XZY)
    Child=(Type=Bone,Name="kin_armBoxes_l_16")
    Parents="((Item=(Type=Bone,Name=\"deform_upperarm_bk_l\"),Weight=0.0005),(Item=(Type=Bone,Name=\"lowerarm_twist_02_l\"),Weight=0.0452),(Item=(Type=Bone,Name=\"upperarm_twist_01_l\"),Weight=0.0603),(Item=(Type=Bone,Name=\"deform_elbow_in_l\"),Weight=0.0612),(Item=(Type=Bone ...(truncated)"
    Weight=1
   )
   > &RigUnit_ParentConstraint(
    AdvancedSettings=(RotationOrderForFilter=XZY)
    Child=(Type=Bone,Name="kin_armBoxes_l_17")
    Parents="((Item=(Type=Bone,Name=\"deform_upperarm_bk_l\"),Weight=0.0026),(Item=(Type=Bone,Name=\"deform_elbow_in_l\"),Weight=0.006),(Item=(Type=Bone,Name=\"lowerarm_twist_02_l\"),Weight=0.0138),(Item=(Type=Bone,Name=\"lowerarm_l\"),Weight=0.0692),(Item=(Type=Bone,Name=\"upp ...(truncated)"
    Weight=1
   )
   &RigUnit_ParentConstraint(
    AdvancedSettings=(RotationOrderForFilter=XZY)
    Child=(Type=Bone,Name="kin_armBoxes_r_1")
    Parents="((Item=(Type=Bone,Name=\"upperarm_twist_01_r\"),Weight=0.0002),(Item=(Type=Bone,Name=\"deform_elbow_out_r\"),Weight=0.0015),(Item=(Type=Bone,Name=\"upperarm_twist_02_r\"),Weight=0.0044),(Item=(Type=Bone,Name=\"lowerarm_twist_01_r\"),Weight=0.0761),(Item=(Type=Bone ...(truncated)"
    Weight=1
    ExecutePin=@RigVMFunction_Sequence.B
   )
   > &RigUnit_ParentConstraint(
    AdvancedSettings=(RotationOrderForFilter=XZY)
    Child=(Type=Bone,Name="kin_armBoxes_r_2")
    Parents="((Item=(Type=Bone,Name=\"upperarm_twist_01_r\"),Weight=0.0011),(Item=(Type=Bone,Name=\"deform_elbow_out_r\"),Weight=0.0087),(Item=(Type=Bone,Name=\"upperarm_twist_02_r\"),Weight=0.0163),(Item=(Type=Bone,Name=\"lowerarm_twist_01_r\"),Weight=0.0413),(Item=(Type=Bone ...(truncated)"
    Weight=1
   )
   > &RigUnit_ParentConstraint(
    AdvancedSettings=(RotationOrderForFilter=XZY)
    Child=(Type=Bone,Name="kin_armBoxes_r_3")
    Parents="((Item=(Type=Bone,Name=\"upperarm_twist_01_r\"),Weight=0.0013),(Item=(Type=Bone,Name=\"lowerarm_twist_01_r\"),Weight=0.024),(Item=(Type=Bone,Name=\"upperarm_twist_02_r\"),Weight=0.0273),(Item=(Type=Bone,Name=\"deform_elbow_out_r\"),Weight=0.0661),(Item=(Type=Bone, ...(truncated)"
    Weight=1
   )
   > &RigUnit_ParentConstraint(
    AdvancedSettings=(RotationOrderForFilter=XZY)
    Child=(Type=Bone,Name="kin_armBoxes_r_4")
    Parents="((Item=(Type=Bone,Name=\"deform_upperarm_fr_r\"),Weight=0.0021),(Item=(Type=Bone,Name=\"deform_elbow_out_r\"),Weight=0.0246),(Item=(Type=Bone,Name=\"lowerarm_twist_02_r\"),Weight=0.0321),(Item=(Type=Bone,Name=\"deform_elbow_in_r\"),Weight=0.0741),(Item=(Type=Bone, ...(truncated)"
    Weight=1
   )
   > &RigUnit_ParentConstraint(
    AdvancedSettings=(RotationOrderForFilter=XZY)
    Child=(Type=Bone,Name="kin_armBoxes_r_5")
    Parents="((Item=(Type=Bone,Name=\"deform_wrist_up_r\"),Weight=0.0006),(Item=(Type=Bone,Name=\"deform_elbow_in_r\"),Weight=0.0032),(Item=(Type=Bone,Name=\"upperarm_twist_02_r\"),Weight=0.0033),(Item=(Type=Bone,Name=\"upperarm_r\"),Weight=0.0647),(Item=(Type=Bone,Name=\"defor ...(truncated)"
    Weight=1
   )
   > &RigUnit_ParentConstraint(
    AdvancedSettings=(RotationOrderForFilter=XZY)
    Child=(Type=Bone,Name="kin_armBoxes_r_6")
    Parents="((Item=(Type=Bone,Name=\"deform_elbow_out_r\"),Weight=0.0003),(Item=(Type=Bone,Name=\"deform_upperarm_fr_r\"),Weight=0.0019),(Item=(Type=Bone,Name=\"lowerarm_twist_02_r\"),Weight=0.0241),(Item=(Type=Bone,Name=\"lowerarm_r\"),Weight=0.0688),(Item=(Type=Bone,Name=\"d ...(truncated)"
    Weight=1
   )
   > &RigUnit_ParentConstraint(
    AdvancedSettings=(RotationOrderForFilter=XZY)
    Child=(Type=Bone,Name="kin_armBoxes_r_7")
    Parents="((Item=(Type=Bone,Name=\"deform_upperarm_fr_r\"),Weight=0.0009),(Item=(Type=Bone,Name=\"deform_elbow_out_r\"),Weight=0.0011),(Item=(Type=Bone,Name=\"lowerarm_twist_02_r\"),Weight=0.0411),(Item=(Type=Bone,Name=\"upperarm_twist_01_r\"),Weight=0.0846),(Item=(Type=Bon ...(truncated)"
    Weight=1
   )
   > &RigUnit_ParentConstraint(
    AdvancedSettings=(RotationOrderForFilter=XZY)
    Child=(Type=Bone,Name="kin_armBoxes_r_8")
    Parents="((Item=(Type=Bone,Name=\"lowerarm_twist_02_r\"),Weight=0.0009),(Item=(Type=Bone,Name=\"deform_upperarm_bk_r\"),Weight=0.001),(Item=(Type=Bone,Name=\"lowerarm_r\"),Weight=0.0112),(Item=(Type=Bone,Name=\"deform_elbow_in_r\"),Weight=0.0162),(Item=(Type=Bone,Name=\"def ...(truncated)"
    Weight=1
   )
   > &RigUnit_ParentConstraint(
    AdvancedSettings=(RotationOrderForFilter=XZY)
    Child=(Type=Bone,Name="kin_armBoxes_r_9")
    Parents="((Item=(Type=Bone,Name=\"deform_upperarm_fr_r\"),Weight=0.0011),(Item=(Type=Bone,Name=\"deform_upperarm_bk_r\"),Weight=0.0451),(Item=(Type=Bone,Name=\"upperarm_r\"),Weight=0.0726),(Item=(Type=Bone,Name=\"upperarm_twist_02_r\"),Weight=0.1661),(Item=(Type=Bone,Name= ...(truncated)"
    Weight=1
   )
   > &RigUnit_ParentConstraint(
    AdvancedSettings=(RotationOrderForFilter=XZY)
    Child=(Type=Bone,Name="kin_armBoxes_r_10")
    Parents="((Item=(Type=Bone,Name=\"deform_upperarm_bk_r\"),Weight=0.0045),(Item=(Type=Bone,Name=\"deform_upperarm_up_r\"),Weight=0.0655),(Item=(Type=Bone,Name=\"deform_upperarm_fr_r\"),Weight=0.0668),(Item=(Type=Bone,Name=\"upperarm_r\"),Weight=0.1081),(Item=(Type=Bone,Name ...(truncated)"
    Weight=1
   )
   > &RigUnit_ParentConstraint(
    AdvancedSettings=(RotationOrderForFilter=XZY)
    Child=(Type=Bone,Name="kin_armBoxes_r_11")
    Parents="((Item=(Type=Bone,Name=\"upperarm_twist_02_r\"),Weight=0.0071),(Item=(Type=Bone,Name=\"deform_elbow_in_r\"),Weight=0.0419),(Item=(Type=Bone,Name=\"lowerarm_twist_01_r\"),Weight=0.0615),(Item=(Type=Bone,Name=\"deform_elbow_out_r\"),Weight=0.084),(Item=(Type=Bone,Na ...(truncated)"
    Weight=1
   )
   > &RigUnit_ParentConstraint(
    AdvancedSettings=(RotationOrderForFilter=XZY)
    Child=(Type=Bone,Name="kin_armBoxes_r_12")
    Parents="((Item=(Type=Bone,Name=\"deform_elbow_in_r\"),Weight=0.0047),(Item=(Type=Bone,Name=\"deform_upperarm_up_r\"),Weight=0.0058),(Item=(Type=Bone,Name=\"deform_upperarm_bk_r\"),Weight=0.0207),(Item=(Type=Bone,Name=\"lowerarm_r\"),Weight=0.0226),(Item=(Type=Bone,Name=\"d ...(truncated)"
    Weight=1
   )
   > &RigUnit_ParentConstraint(
    AdvancedSettings=(RotationOrderForFilter=XZY)
    Child=(Type=Bone,Name="kin_armBoxes_r_13")
    Parents="((Item=(Type=Bone,Name=\"deform_upperarm_fr_r\"),Weight=0.0021),(Item=(Type=Bone,Name=\"deform_elbow_out_r\"),Weight=0.0246),(Item=(Type=Bone,Name=\"lowerarm_twist_02_r\"),Weight=0.0321),(Item=(Type=Bone,Name=\"deform_elbow_in_r\"),Weight=0.0741),(Item=(Type=Bone, ...(truncated)"
    Weight=1
   )
   > &RigUnit_ParentConstraint(
    AdvancedSettings=(RotationOrderForFilter=XZY)
    Child=(Type=Bone,Name="kin_armBoxes_r_14")
    Parents="((Item=(Type=Bone,Name=\"lowerarm_twist_01_r\"),Weight=0.0009),(Item=(Type=Bone,Name=\"deform_elbow_out_r\"),Weight=0.0168),(Item=(Type=Bone,Name=\"upperarm_twist_01_r\"),Weight=0.0356),(Item=(Type=Bone,Name=\"lowerarm_twist_02_r\"),Weight=0.0751),(Item=(Type=Bone ...(truncated)"
    Weight=1
   )
   > &RigUnit_ParentConstraint(
    AdvancedSettings=(RotationOrderForFilter=XZY)
    Child=(Type=Bone,Name="kin_armBoxes_r_15")
    Parents="((Item=(Type=Bone,Name=\"deform_upperarm_bk_r\"),Weight=0.0007),(Item=(Type=Bone,Name=\"lowerarm_twist_02_r\"),Weight=0.042),(Item=(Type=Bone,Name=\"upperarm_twist_01_r\"),Weight=0.0571),(Item=(Type=Bone,Name=\"deform_elbow_in_r\"),Weight=0.068),(Item=(Type=Bone,N ...(truncated)"
    Weight=1
   )
   > &RigUnit_ParentConstraint(
    AdvancedSettings=(RotationOrderForFilter=XZY)
    Child=(Type=Bone,Name="kin_armBoxes_r_16")
    Parents="((Item=(Type=Bone,Name=\"deform_upperarm_bk_r\"),Weight=0.0034),(Item=(Type=Bone,Name=\"deform_elbow_in_r\"),Weight=0.0125),(Item=(Type=Bone,Name=\"lowerarm_twist_02_r\"),Weight=0.0175),(Item=(Type=Bone,Name=\"lowerarm_r\"),Weight=0.076),(Item=(Type=Bone,Name=\"upp ...(truncated)"
    Weight=1
   )
   > &RigUnit_ParentConstraint(
    AdvancedSettings=(RotationOrderForFilter=XZY)
    Child=(Type=Bone,Name="kin_armBoxes_r_17")
    Parents="((Item=(Type=Bone,Name=\"deform_upperarm_bk_r\"),Weight=0.003),(Item=(Type=Bone,Name=\"deform_elbow_in_r\"),Weight=0.007),(Item=(Type=Bone,Name=\"lowerarm_twist_02_r\"),Weight=0.016),(Item=(Type=Bone,Name=\"lowerarm_r\"),Weight=0.0721),(Item=(Type=Bone,Name=\"upper ...(truncated)"
    Weight=1
   )
   &RigUnit_ParentConstraint(
    AdvancedSettings=(RotationOrderForFilter=XZY)
    Child=(Type=Bone,Name="kin_left_torso_1")
    Parents=((Item=(Type=Bone,Name="clavicle_l"),Weight=1.0),(Item=(Type=Bone,Name="None"),Weight=1))
    Weight=1
    ExecutePin=@RigVMFunction_Sequence.C
   )
   > &RigUnit_ParentConstraint(
    AdvancedSettings=(RotationOrderForFilter=XZY)
    Child=(Type=Bone,Name="kin_left_torso_2")
    Parents=((Item=(Type=Bone,Name="clavicle_l"),Weight=1.0))
    Weight=1
   )
   > &RigUnit_ParentConstraint(
    AdvancedSettings=(RotationOrderForFilter=XZY)
    Child=(Type=Bone,Name="kin_left_torso_3")
    Parents=((Item=(Type=Bone,Name="clavicle_l"),Weight=1.0))
    Weight=1
   )
   > &RigUnit_ParentConstraint(
    AdvancedSettings=(RotationOrderForFilter=XZY)
    Child=(Type=Bone,Name="kin_left_torso_4")
    Parents=((Item=(Type=Bone,Name="clavicle_l"),Weight=1.0))
    Weight=1
   )
   > &RigUnit_ParentConstraint(
    AdvancedSettings=(RotationOrderForFilter=XZY)
    Child=(Type=Bone,Name="kin_front_torso_1")
    Filter=(TranslationFilter=(bX=True,bY=True,bZ=True),RotationFilter=(bX=False,bY=False,bZ=False),ScaleFilter=(bX=True,bY=True,bZ=True))
    Parents="((Item=(Type=Bone,Name=\"spine_01\"),Weight=0),(Item=(Type=Bone,Name=\"clavicle_r\"),Weight=0),(Item=(Type=Bone,Name=\"spine_02\"),Weight=0),(Item=(Type=Bone,Name=\"clavicle_l\"),Weight=0.3),(Item=(Type=Bone,Name=\"spine_03\"),Weight=0.1),(Item=(Type=Bone,Name=\"spin ...(truncated)"
    Weight=1
   )
   > &RigUnit_ParentConstraint(
    AdvancedSettings=(RotationOrderForFilter=XZY)
    Child=(Type=Bone,Name="kin_front_torso_2")
    Filter=(TranslationFilter=(bX=True,bY=True,bZ=True),RotationFilter=(bX=False,bY=False,bZ=False),ScaleFilter=(bX=True,bY=True,bZ=True))
    Parents="((Item=(Type=Bone,Name=\"pelvis\"),Weight=0),(Item=(Type=Bone,Name=\"spine_01\"),Weight=0),(Item=(Type=Bone,Name=\"clavicle_l\"),Weight=0),(Item=(Type=Bone,Name=\"spine_02\"),Weight=0.2),(Item=(Type=Bone,Name=\"spine_05\"),Weight=0.2),(Item=(Type=Bone,Name=\"spine_04 ...(truncated)"
    Weight=1
   )
   > &RigUnit_ParentConstraint(
    AdvancedSettings=(RotationOrderForFilter=XZY)
    Child=(Type=Bone,Name="kin_front_torso_3")
    Filter=(TranslationFilter=(bX=True,bY=True,bZ=True),RotationFilter=(bX=False,bY=False,bZ=False),ScaleFilter=(bX=True,bY=True,bZ=True))
    Parents="((Item=(Type=Bone,Name=\"clavicle_l\"),Weight=0.0002),(Item=(Type=Bone,Name=\"pelvis\"),Weight=0.005),(Item=(Type=Bone,Name=\"spine_05\"),Weight=0.0259),(Item=(Type=Bone,Name=\"spine_01\"),Weight=0.043),(Item=(Type=Bone,Name=\"spine_04\"),Weight=0.1056),(Item=(Type= ...(truncated)"
    Weight=1
   )
   > &RigUnit_ParentConstraint(
    AdvancedSettings=(RotationOrderForFilter=XZY)
    Child=(Type=Bone,Name="kin_front_torso_4")
    Filter=(TranslationFilter=(bX=True,bY=True,bZ=True),RotationFilter=(bX=False,bY=False,bZ=False),ScaleFilter=(bX=True,bY=True,bZ=True))
    Parents="((Item=(Type=Bone,Name=\"spine_01\"),Weight=0),(Item=(Type=Bone,Name=\"spine_02\"),Weight=0),(Item=(Type=Bone,Name=\"clavicle_l\"),Weight=0),(Item=(Type=Bone,Name=\"clavicle_r\"),Weight=0.2),(Item=(Type=Bone,Name=\"spine_03\"),Weight=0.05),(Item=(Type=Bone,Name=\"spi ...(truncated)"
    Weight=1
   )
   > &RigUnit_ParentConstraint(
    AdvancedSettings=(RotationOrderForFilter=XZY)
    Child=(Type=Bone,Name="kin_front_torso_5")
    Filter=(TranslationFilter=(bX=True,bY=True,bZ=True),RotationFilter=(bX=False,bY=False,bZ=False),ScaleFilter=(bX=True,bY=True,bZ=True))
    Parents="((Item=(Type=Bone,Name=\"pelvis\"),Weight=0.0004),(Item=(Type=Bone,Name=\"clavicle_r\"),Weight=0.0041),(Item=(Type=Bone,Name=\"spine_01\"),Weight=0.006),(Item=(Type=Bone,Name=\"spine_02\"),Weight=0.1258),(Item=(Type=Bone,Name=\"spine_04\"),Weight=0.2083),(Item=(Type ...(truncated)"
    Weight=1
   )
   > &RigUnit_ParentConstraint(
    AdvancedSettings=(RotationOrderForFilter=XZY)
    Child=(Type=Bone,Name="kin_front_torso_6")
    Filter=(TranslationFilter=(bX=True,bY=True,bZ=True),RotationFilter=(bX=False,bY=False,bZ=False),ScaleFilter=(bX=True,bY=True,bZ=True))
    Parents="((Item=(Type=Bone,Name=\"pelvis\"),Weight=0.0009),(Item=(Type=Bone,Name=\"clavicle_r\"),Weight=0.001),(Item=(Type=Bone,Name=\"spine_01\"),Weight=0.0117),(Item=(Type=Bone,Name=\"spine_05\"),Weight=0.0739),(Item=(Type=Bone,Name=\"spine_04\"),Weight=0.2101),(Item=(Type ...(truncated)"
    Weight=1
   )
   > &RigUnit_ParentConstraint(
    AdvancedSettings=(RotationOrderForFilter=XZY)
    Child=(Type=Bone,Name="kin_front_torso_7")
    Filter=(TranslationFilter=(bX=True,bY=True,bZ=True),RotationFilter=(bX=False,bY=False,bZ=False),ScaleFilter=(bX=True,bY=True,bZ=True))
    Parents="((Item=(Type=Bone,Name=\"pelvis\"),Weight=0.0002),(Item=(Type=Bone,Name=\"spine_01\"),Weight=0.0035),(Item=(Type=Bone,Name=\"clavicle_r\"),Weight=0.0187),(Item=(Type=Bone,Name=\"spine_02\"),Weight=0.0919),(Item=(Type=Bone,Name=\"spine_05\"),Weight=0.1494),(Item=(Typ ...(truncated)"
    Weight=1
   )
   > &RigUnit_ParentConstraint(
    AdvancedSettings=(RotationOrderForFilter=XZY)
    Child=(Type=Bone,Name="kin_front_torso_8")
    Filter=(TranslationFilter=(bX=True,bY=True,bZ=True),RotationFilter=(bX=False,bY=False,bZ=False),ScaleFilter=(bX=True,bY=True,bZ=True))
    Parents="((Item=(Type=Bone,Name=\"spine_02\"),Weight=0.0001),(Item=(Type=Bone,Name=\"spine_03\"),Weight=0.0022),(Item=(Type=Bone,Name=\"clavicle_l\"),Weight=0.0056),(Item=(Type=Bone,Name=\"spine_04\"),Weight=0.0119),(Item=(Type=Bone,Name=\"clavicle_r\"),Weight=0.0736),(Item= ...(truncated)"
    Weight=1
   )
   > &RigUnit_ParentConstraint(
    AdvancedSettings=(RotationOrderForFilter=XZY)
    Child=(Type=Bone,Name="kin_front_torso_9")
    Filter=(TranslationFilter=(bX=True,bY=True,bZ=True),RotationFilter=(bX=False,bY=False,bZ=False),ScaleFilter=(bX=True,bY=True,bZ=True))
    Parents="((Item=(Type=Bone,Name=\"spine_01\"),Weight=0),(Item=(Type=Bone,Name=\"spine_02\"),Weight=0),(Item=(Type=Bone,Name=\"clavicle_r\"),Weight=0.25),(Item=(Type=Bone,Name=\"spine_03\"),Weight=0.05),(Item=(Type=Bone,Name=\"spine_04\"),Weight=0.2),(Item=(Type=Bone,Name=\"sp ...(truncated)"
    Weight=1
   )
   > &RigUnit_ParentConstraint(
    AdvancedSettings=(RotationOrderForFilter=XZY)
    Child=(Type=Bone,Name="kin_back_torso_1")
    Filter=(TranslationFilter=(bX=True,bY=True,bZ=True),RotationFilter=(bX=False,bY=False,bZ=False),ScaleFilter=(bX=True,bY=True,bZ=True))
    Parents="((Item=(Type=Bone,Name=\"spine_01\"),Weight=0),(Item=(Type=Bone,Name=\"clavicle_r\"),Weight=0),(Item=(Type=Bone,Name=\"spine_02\"),Weight=0),(Item=(Type=Bone,Name=\"clavicle_l\"),Weight=0.1),(Item=(Type=Bone,Name=\"spine_03\"),Weight=0.2),(Item=(Type=Bone,Name=\"spin ...(truncated)"
    Weight=1
   )
   > &RigUnit_ParentConstraint(
    AdvancedSettings=(RotationOrderForFilter=XZY)
    Child=(Type=Bone,Name="kin_back_torso_2")
    Filter=(TranslationFilter=(bX=True,bY=True,bZ=True),RotationFilter=(bX=False,bY=False,bZ=False),ScaleFilter=(bX=True,bY=True,bZ=True))
    Parents="((Item=(Type=Bone,Name=\"spine_01\"),Weight=0),(Item=(Type=Bone,Name=\"clavicle_r\"),Weight=0),(Item=(Type=Bone,Name=\"spine_02\"),Weight=0),(Item=(Type=Bone,Name=\"clavicle_l\"),Weight=0.15),(Item=(Type=Bone,Name=\"spine_03\"),Weight=0.1),(Item=(Type=Bone,Name=\"spi ...(truncated)"
    Weight=1
   )
   > &RigUnit_ParentConstraint(
    AdvancedSettings=(RotationOrderForFilter=XZY)
    Child=(Type=Bone,Name="kin_back_torso_3")
    Filter=(TranslationFilter=(bX=True,bY=True,bZ=True),RotationFilter=(bX=False,bY=False,bZ=False),ScaleFilter=(bX=True,bY=True,bZ=True))
    Parents=((Item=(Type=Bone,Name="spine_03"),Weight=0),(Item=(Type=Bone,Name="spine_04"),Weight=0.05),(Item=(Type=Bone,Name="clavicle_l"),Weight=0.05),(Item=(Type=Bone,Name="clavicle_r"),Weight=0.1),(Item=(Type=Bone,Name="spine_05"),Weight=0.8))
    Weight=1
   )
   > &RigUnit_ParentConstraint(
    AdvancedSettings=(RotationOrderForFilter=XZY)
    Child=(Type=Bone,Name="kin_back_torso_4")
    Filter=(TranslationFilter=(bX=True,bY=True,bZ=True),RotationFilter=(bX=False,bY=False,bZ=False),ScaleFilter=(bX=True,bY=True,bZ=True))
    Parents="((Item=(Type=Bone,Name=\"spine_01\"),Weight=0),(Item=(Type=Bone,Name=\"spine_02\"),Weight=0),(Item=(Type=Bone,Name=\"clavicle_l\"),Weight=0),(Item=(Type=Bone,Name=\"clavicle_r\"),Weight=0.15),(Item=(Type=Bone,Name=\"spine_03\"),Weight=0.05),(Item=(Type=Bone,Name=\"sp ...(truncated)"
    Weight=1
   )
   > &RigUnit_ParentConstraint(
    AdvancedSettings=(RotationOrderForFilter=XZY)
    Child=(Type=Bone,Name="kin_back_torso_5")
    Filter=(TranslationFilter=(bX=True,bY=True,bZ=True),RotationFilter=(bX=False,bY=False,bZ=False),ScaleFilter=(bX=True,bY=True,bZ=True))
    Parents="((Item=(Type=Bone,Name=\"pelvis\"),Weight=0),(Item=(Type=Bone,Name=\"spine_01\"),Weight=0),(Item=(Type=Bone,Name=\"clavicle_l\"),Weight=0.1),(Item=(Type=Bone,Name=\"spine_02\"),Weight=0),(Item=(Type=Bone,Name=\"spine_05\"),Weight=0.2),(Item=(Type=Bone,Name=\"spine_03 ...(truncated)"
    Weight=1
   )
   > &RigUnit_ParentConstraint(
    AdvancedSettings=(RotationOrderForFilter=XZY)
    Child=(Type=Bone,Name="kin_back_torso_6")
    Filter=(TranslationFilter=(bX=True,bY=True,bZ=True),RotationFilter=(bX=False,bY=False,bZ=False),ScaleFilter=(bX=True,bY=True,bZ=True))
    Parents="((Item=(Type=Bone,Name=\"pelvis\"),Weight=0.0003),(Item=(Type=Bone,Name=\"spine_01\"),Weight=0.0041),(Item=(Type=Bone,Name=\"clavicle_r\"),Weight=0.0042),(Item=(Type=Bone,Name=\"spine_02\"),Weight=0.0817),(Item=(Type=Bone,Name=\"spine_05\"),Weight=0.0914),(Item=(Typ ...(truncated)"
    Weight=1
   )
   > &RigUnit_ParentConstraint(
    AdvancedSettings=(RotationOrderForFilter=XZY)
    Child=(Type=Bone,Name="kin_back_torso_7")
    Filter=(TranslationFilter=(bX=True,bY=True,bZ=True),RotationFilter=(bX=False,bY=False,bZ=False),ScaleFilter=(bX=True,bY=True,bZ=True))
    Parents="((Item=(Type=Bone,Name=\"clavicle_r\"),Weight=0.0006),(Item=(Type=Bone,Name=\"pelvis\"),Weight=0.0073),(Item=(Type=Bone,Name=\"spine_05\"),Weight=0.0149),(Item=(Type=Bone,Name=\"spine_01\"),Weight=0.0483),(Item=(Type=Bone,Name=\"spine_04\"),Weight=0.1487),(Item=(Typ ...(truncated)"
    Weight=1
   )
   > &RigUnit_ParentConstraint(
    AdvancedSettings=(RotationOrderForFilter=XZY)
    Child=(Type=Bone,Name="kin_back_torso_8")
    Filter=(TranslationFilter=(bX=True,bY=True,bZ=True),RotationFilter=(bX=False,bY=False,bZ=False),ScaleFilter=(bX=True,bY=True,bZ=True))
    Parents="((Item=(Type=Bone,Name=\"spine_05\"),Weight=0.0007),(Item=(Type=Bone,Name=\"spine_04\"),Weight=0.0234),(Item=(Type=Bone,Name=\"pelvis\"),Weight=0.0423),(Item=(Type=Bone,Name=\"spine_01\"),Weight=0.1582),(Item=(Type=Bone,Name=\"spine_03\"),Weight=0.2471),(Item=(Type= ...(truncated)"
    Weight=1
   )
   > &RigUnit_ParentConstraint(
    AdvancedSettings=(RotationOrderForFilter=XZY)
    Child=(Type=Bone,Name="kin_back_torso_9")
    Filter=(TranslationFilter=(bX=True,bY=True,bZ=True),RotationFilter=(bX=False,bY=False,bZ=False),ScaleFilter=(bX=True,bY=True,bZ=True))
    Parents="((Item=(Type=Bone,Name=\"spine_01\"),Weight=0),(Item=(Type=Bone,Name=\"clavicle_l\"),Weight=0),(Item=(Type=Bone,Name=\"spine_02\"),Weight=0),(Item=(Type=Bone,Name=\"spine_03\"),Weight=0.05),(Item=(Type=Bone,Name=\"spine_04\"),Weight=0.15),(Item=(Type=Bone,Name=\"spin ...(truncated)"
    Weight=1
   )
   > &RigUnit_ParentConstraint(
    AdvancedSettings=(RotationOrderForFilter=XZY)
    Child=(Type=Bone,Name="kin_back_torso_10")
    Filter=(TranslationFilter=(bX=True,bY=True,bZ=True),RotationFilter=(bX=False,bY=False,bZ=False),ScaleFilter=(bX=True,bY=True,bZ=True))
    Parents=((Item=(Type=Bone,Name="spine_03"),Weight=0.0002),(Item=(Type=Bone,Name="spine_04"),Weight=0.004),(Item=(Type=Bone,Name="clavicle_r"),Weight=0.0119),(Item=(Type=Bone,Name="clavicle_l"),Weight=0.0324),(Item=(Type=Bone,Name="spine_05"),Weight=0.9515))
    Weight=1
   )
   > &RigUnit_ParentConstraint(
    AdvancedSettings=(RotationOrderForFilter=XZY)
    Child=(Type=Bone,Name="kin_right_torso_1")
    Parents=((Item=(Type=Bone,Name="clavicle_r"),Weight=1.0))
    Weight=1
   )
   > &RigUnit_ParentConstraint(
    AdvancedSettings=(RotationOrderForFilter=XZY)
    Child=(Type=Bone,Name="kin_right_torso_2")
    Parents=((Item=(Type=Bone,Name="clavicle_r"),Weight=1.0))
    Weight=1
   )
   > &RigUnit_ParentConstraint(
    AdvancedSettings=(RotationOrderForFilter=XZY)
    Child=(Type=Bone,Name="kin_right_torso_3")
    Parents=((Item=(Type=Bone,Name="clavicle_r"),Weight=1.0))
    Weight=1
   )
   > &RigUnit_ParentConstraint(
    AdvancedSettings=(RotationOrderForFilter=XZY)
    Child=(Type=Bone,Name="kin_right_torso_4")
    Parents="((Item=(Type=Bone,Name=\"spine_02\"),Weight=0.0021),(Item=(Type=Bone,Name=\"clavicle_l\"),Weight=0.0038),(Item=(Type=Bone,Name=\"spine_03\"),Weight=0.0224),(Item=(Type=Bone,Name=\"spine_04\"),Weight=0.0638),(Item=(Type=Bone,Name=\"clavicle_r\"),Weight=0.1143),(Item= ...(truncated)"
    Weight=1
    bMaintainOffset=true
   )
   &RigUnit_ModifyTransforms(
    ItemToModify="((Item=(Type=Bone,Name=\"kin_front_torso_5\"),Transform=(Rotation=(X=0,Y=0,Z=0,W=-1),Translation=(X=0,Y=0,Z=1),Scale3D=(X=1,Y=1,Z=1))),(Item=(Type=Bone,Name=\"kin_front_torso_4\"),Transform=(Rotation=(X=0,Y=0,Z=0,W=-1),Translation=(X=0,Y=0,Z=0.3),Scale3D=(X=1, ...(truncated)"
    Weight=1
    WeightMinimum=0
    ExecutePin=@RigVMFunction_Sequence.D
   )
   > &RigUnit_ParentConstraint(
    AdvancedSettings=(RotationOrderForFilter=XZY)
    Child=(Type=Bone,Name="armor_collar")
    Parents=((Item=(Type=Bone,Name="neck_01"),Weight=0.5),(Item=(Type=Bone,Name="neck_02"),Weight=0.5))
    Weight=1
   )
   Set Bone Translation Osscilation_6 = RigVMFunctionReferenceNode(Offset=60, Scale=0.5, Speed=1, Target=(Type=Bone,Name="kin_back_torso_1"), ExecuteContext=@RigVMFunction_Sequence.F)
   > Set Bone Translation Osscilation_7 = RigVMFunctionReferenceNode(Offset=10, Scale=0.8, Speed=1, Target=(Type=Bone,Name="kin_back_torso_2"))
   > Set Bone Translation Osscilation_8 = RigVMFunctionReferenceNode(Offset=20, Scale=0.75, Speed=1, Target=(Type=Bone,Name="kin_back_torso_3"))
   > Set Bone Translation Osscilation_9 = RigVMFunctionReferenceNode(Offset=-15, Scale=1, Speed=1, Target=(Type=Bone,Name="kin_back_torso_4"))
   > Set Bone Translation Osscilation_10 = RigVMFunctionReferenceNode(Offset=-30, Scale=0.5, Speed=1, Target=(Type=Bone,Name="kin_back_torso_5"))
   > Set Bone Translation Osscilation_11 = RigVMFunctionReferenceNode(Offset=60, Scale=0.5, Speed=1, Target=(Type=Bone,Name="kin_back_torso_6"))
   > Set Bone Translation Osscilation_12 = RigVMFunctionReferenceNode(Offset=20, Scale=0.75, Speed=1, Target=(Type=Bone,Name="kin_back_torso_7"))
   > Set Bone Translation Osscilation_20 = RigVMFunctionReferenceNode(Offset=75, Scale=0.5, Speed=1, Target=(Type=Bone,Name="kin_back_torso_9"))
   Set Bone Translation Osscilation_13 = RigVMFunctionReferenceNode(Offset=0, Scale=1, Speed=1, Target=(Type=Bone,Name="kin_left_torso_1"), ExecuteContext=@RigVMFunction_Sequence.G)
   > Set Bone Translation Osscilation_14 = RigVMFunctionReferenceNode(Offset=10, Scale=0.8, Speed=1, Target=(Type=Bone,Name="kin_left_torso_2"))
   > Set Bone Translation Osscilation_15 = RigVMFunctionReferenceNode(Offset=20, Scale=0.75, Speed=1, Target=(Type=Bone,Name="kin_left_torso_3"))
   > Set Bone Translation Osscilation_19 = RigVMFunctionReferenceNode(Offset=50, Scale=1, Speed=1, Target=(Type=Bone,Name="kin_left_torso_4"))
   Set Bone Translation Osscilation_16 = RigVMFunctionReferenceNode(Offset=-15, Scale=1, Speed=1, Target=(Type=Bone,Name="kin_right_torso_1"), ExecuteContext=@RigVMFunction_Sequence.H)
   > Set Bone Translation Osscilation_17 = RigVMFunctionReferenceNode(Offset=-30, Scale=0.5, Speed=1, Target=(Type=Bone,Name="kin_right_torso_2"))
   > Set Bone Translation Osscilation_18 = RigVMFunctionReferenceNode(Offset=60, Scale=0.5, Speed=1, Target=(Type=Bone,Name="kin_right_torso_3"))
   RigVMFunctionReturnNode(ExecutePin=@"Set Bone Translation Osscilation_18".ExecuteContext)
   Set Bone Translation Osscilation = RigVMFunctionReferenceNode(Offset=1, Scale=0.5, Speed=1, Target=(Type=Bone,Name="kin_front_torso_1"), ExecuteContext=@RigVMFunction_Sequence.E)
   > Set Bone Translation Osscilation_1 = RigVMFunctionReferenceNode(Offset=10, Scale=0.8, Speed=1, Target=(Type=Bone,Name="kin_front_torso_2"))
   > Set Bone Translation Osscilation_2 = RigVMFunctionReferenceNode(Offset=20, Scale=0.75, Speed=1, Target=(Type=Bone,Name="kin_front_torso_3"))
   > Set Bone Translation Osscilation_3 = RigVMFunctionReferenceNode(Offset=-15, Scale=1, Speed=1, Target=(Type=Bone,Name="kin_front_torso_4"))
   > Set Bone Translation Osscilation_4 = RigVMFunctionReferenceNode(Offset=-30, Scale=0.5, Speed=1, Target=(Type=Bone,Name="kin_front_torso_5"))
   > Set Bone Translation Osscilation_22 = RigVMFunctionReferenceNode(Offset=0, Scale=0.75, Speed=1, Target=(Type=Bone,Name="kin_front_torso_6"))
   > Set Bone Translation Osscilation_5 = RigVMFunctionReferenceNode(Offset=60, Scale=0.5, Speed=1, Target=(Type=Bone,Name="kin_front_torso_7"))
   > Set Bone Translation Osscilation_21 = RigVMFunctionReferenceNode(Offset=75, Scale=0.5, Speed=1, Target=(Type=Bone,Name="kin_back_torso_8"))
  ]}
 New Function_ContainedGraph {nodes [
   RigVMFunctionEntryNode()
   RigVMFunction_MathBoolToInteger = &RigVMFunction_MathBoolToInteger(Value=RigVMVariableNode() {variable=onlyTwist}.Value)
   RigVMFunction_MathQuaternionSwingTwist = &RigVMFunction_MathQuaternionSwingTwist(
    Input=&RigUnit_GetRelativeTransformForItem(
     bChildInitial=False
     bParentInitial=False
     Child=RigVMVariableNode() {variable=child}.Value
     Parent=RigVMVariableNode() {variable=parent}.Value
    ).RelativeTransform.Rotation
    TwistAxis=RigVMVariableNode() {variable=twistAxis}.Value
   )
   n1 = RigVMDispatchNode(
    Values=((),())
    Index=@RigVMFunction_MathBoolToInteger.Result
    Values.0=@RigVMFunction_MathQuaternionSwingTwist.Swing
    Values.1=@RigVMFunction_MathQuaternionSwingTwist.Twist
   ) {template="DISPATCH_RigVMDispatch_SelectInt32(in Index,in Values,out Result)"}
   > RigVMVariableNode(Value=@n1.Result) {variable=entry_outputQuat}
   RigVMFunction_MathQuaternionToEuler = &RigVMFunction_MathQuaternionToEuler(Value=@n1.Result, RotationOrder=RigVMVariableNode() {variable=Rotation Order}.Value)
   > DISPATCH_RigVMDispatch_Print = RigVMDispatchNode(
    ScreenColor=(R=1,G=1,B=1,A=1)
    ScreenDuration=1.5
    Value=@RigVMFunction_MathQuaternionToEuler.Result
    Enabled=RigVMVariableNode() {variable=debugRotation}.Value
   ) {template="DISPATCH_RigVMDispatch_Print(in Prefix,in Value,in Enabled,in ScreenDuration,in ScreenColor)"}
   Cast = RigVMDispatchNode(Value=RigVMVariableNode() {variable=triggerAxis}.Value) {template="DISPATCH_RigVMDispatch_CastEnumToInt(in Value,out Result)"}
   DISPATCH_RigVMDispatch_SelectInt32 = RigVMDispatchNode(
    Values=(0,0,0)
    Index=@Cast.Result
    Values.0=@RigVMFunction_MathQuaternionToEuler.Result.X
    Values.1=@RigVMFunction_MathQuaternionToEuler.Result.Y
    Values.2=@RigVMFunction_MathQuaternionToEuler.Result.Z
   ) {template="DISPATCH_RigVMDispatch_SelectInt32(in Index,in Values,out Result)"}
   n2 = RigVMVariableNode() {variable=rotationMinMax}
   n3 = RigVMVariableNode() {variable=outputMinMax}
   Remap = &RigVMFunction_MathDoubleRemap(
    bClamp=False
    Value=@DISPATCH_RigVMDispatch_SelectInt32.Result
    SourceMinimum=@n2.Value.X
    SourceMaximum=@n2.Value.Y
    TargetMinimum=@n3.Value.X
    TargetMaximum=@n3.Value.Y
   )
   Absolute = &RigVMFunction_MathDoubleAbs(Value=@Remap.Result)
   n4 = RigVMVariableNode() {variable=absoluteValue}
   Multiply = &RigVMFunction_MathDoubleMul(
    A=&RigVMFunction_MathDoubleClamp(Value=RigVMDispatchNode(True=@Absolute.Result, False=@Remap.Result, Condition=@n4.Value) {template="DISPATCH_RigVMDispatch_If(in Condition,in True,in False,out Result)"}.Result).Result
    B=RigVMVariableNode() {variable=premult}.Value
   )
   > RigVMFunctionReturnNode(alpha=@Multiply.Result, rawQuat=RigVMVariableNode() {variable=entry_outputQuat}.Value)
  ]}
 New Function_ContainedGraph {nodes [
   n1 = RigVMFunctionEntryNode()
   RigVMFunction_Timeline = /Script/RigVM.RigVMFunction_Timeline(Speed=@n1.Speed) {is=struct RigVMUnitNode}
   Add = &RigVMFunction_MathDoubleAdd(B=@n1.Offset, A=@RigVMFunction_Timeline.Time)
   Sin = /Script/RigVM.RigVMFunction_MathDoubleSin(Value=@Add.Result) {is=struct RigVMUnitNode, template="Sin::Execute(in Value,out Result)"}
   Multiply = &RigVMFunction_MathDoubleMul(B=@n1.Scale, A=@Sin.Result)
   Cos = /Script/RigVM.RigVMFunction_MathDoubleCos(Value=@Add.Result) {is=struct RigVMUnitNode, template="Cos::Execute(in Value,out Result)"}
   n2 = RigVMVariableNode() {variable=Target}
   > RigUnit_SetRelativeTransformForItem = /Script/ControlRig.RigUnit_SetRelativeTransformForItem(
    Value=(Rotation=(X=0,Y=0,Z=0,W=1),Translation=(X=0,Y=0,Z=0),Scale3D=(X=1,Y=1,Z=1))
    Weight=1
    bParentInitial=False
    bPropagateToChildren=True
    Value.Translation.Z=@Multiply.Result
    Value.Translation.Y=&RigVMFunction_MathDoubleMul(B=0.5, A=@Cos.Result).Result
    Child=@n2.Value
    Parent=@n2.Value
   ) {
    is=struct RigVMUnitNode
    template="Set Relative Transform::Execute(in Child,in Parent,in bParentInitial,in Value,in Weight,in bPropagateToChildren)"
   }
   > RigVMFunctionReturnNode()
   n3 = RigVMVariableNode() {variable=Target}
   &RigUnit_SetRelativeTransformForItem(
    ExecutePin=()
    Value=(Rotation=(X=0,Y=0,Z=0,W=1),Translation=(X=0,Y=0,Z=0),Scale3D=(X=1,Y=1,Z=1))
    Weight=1
    bParentInitial=False
    bPropagateToChildren=True
    Child=@n3.Value
    Parent=@n3.Value
   ) {template="Set Relative Transform::Execute(in Child,in Parent,in bParentInitial,in Value,in Weight,in bPropagateToChildren)"}
   RigUnit_GetRelativeTransformForItem = &RigUnit_GetRelativeTransformForItem(bChildInitial=False, bParentInitial=False, Child=@n1.Target, Parent=@n1.Target)
   SpringInterp = /Script/ControlRig.RigUnit_SpringInterpVectorV2(
    CriticalDamping=1
    Force=(X=0,Y=0,Z=0)
    Strength=1
    TargetVelocityAmount=0.1
    bInitializeFromTarget=False
    bUseCurrentInput=false
    Current=@RigUnit_GetRelativeTransformForItem.RelativeTransform.Translation
    Target=@RigUnit_GetRelativeTransformForItem.RelativeTransform.Translation
   ) {
    is=struct RigVMUnitNode
    template="SpringInterp::Execute(in Target,in Strength,in CriticalDamping,in Force,in bUseCurrentInput,in Current,in TargetVelocityAmount,in bInitializeFromTarget,out Result,out Velocity)"
   }
  ]}
 RigVMFunctionLibrary {nodes [
   GetAlphaFromQuatDelta = RigVMCollapseNode(
    ExecuteContext=()
    Rotation Order=ZYX
    child=(Type=Bone,Name="None")
    outputMinMax=(X=0,Y=1)
    parent=(Type=Bone,Name="None")
    premult=1
    rotationMinMax=(X=0,Y=0)
    twistAxis=(X=1,Y=0,Z=0)
   )
   Set Bone Translation Osscilation = RigVMCollapseNode(ExecuteContext=(), Offset=0, Scale=1, Speed=1, Target=(Type=None,Name="None"))
  ]}
 RigVMFunction_Sequence_1_ContainedGraph {nodes [
   n1 = RigVMFunctionEntryNode()
   RigVMFunction_Sequence = &RigVMFunction_Sequence(ExecuteContext=@n1.ExecuteContext)
   n2 = &RigVMFunction_Sequence(ExecuteContext=@RigVMFunction_Sequence.B)
   n3 = &RigVMFunction_Sequence(ExecuteContext=@n2.B)
   n4 = &RigVMFunction_Sequence(ExecuteContext=@n3.B)
   n5 = &RigVMFunction_Sequence(ExecuteContext=@n4.B)
   n6 = &RigVMFunction_Sequence(ExecuteContext=@n5.B)
   n7 = &RigVMFunction_Sequence(ExecuteContext=@n6.B)
   RigVMFunctionReturnNode(A=@RigVMFunction_Sequence.A, B=@n2.A, C=@n3.A, D=@n4.A, E=@n5.A, F=@n6.A, G=@n7.A, H=@n7.B)
  ]}
 RigVMModel {nodes [
   RigUnit_BeginExecution = &RigUnit_BeginExecution()
   FN_DeformRig_FNC = RigVMFunctionReferenceNode(
    DeformRig_ Curve="(deformRig_l_upperarm_up_alpha_72_#1=False,deformRig_r_upperarm_up_alpha_73_#2=False,deformRig_l_upperarm_fr_alpha_74_#3=False,deformRig_r_upperarm_fr_alpha_75_CB51B8 ...(truncated)"
    DeformRig_Bool="(upperarm_1_#4=True,elbow_3_#5=True,wrist_7_#6=True,groin_8_#7=True,glute_17_#8=True,knee_9_40C408324CA261 ...(truncated)"
    DeformRig_Debug=(upperarm_debug_18_#4=False,wrist_debug_20_#6=False,lat_debug_26_#9=False,groinGlute_debug_27_#7=False)
    DeformRig_Value="(upperarm_up_18_#10=(Rotation=(X=0,Y=0,Z=0,W=1),Translation=(X=0,Y=0,Z=8),Scale3D=(X=0.7,Y=1,Z=1)),upperarm_fwd_fr_19_#11=(Rotation=(X=0,Y=0,Z=0,W=1),Translation=(X=0,Y=10,Z=0),Scale3D=(X=0.5,Y=1,Z= ...(truncated)"
    ExecuteContext=@RigUnit_BeginExecution.ExecutePin
   )
   constrains = RigVMCollapseNode(ExecutePin=@FN_DeformRig_FNC.ExecuteContext)
   GetAlphaFromQuatDelta = RigVMFunctionReferenceNode(
    Rotation Order=ZYX
    child=(Type=Bone,Name="head")
    debugRotation=false
    outputMinMax=(X=0,Y=1)
    parent=(Type=Null,Name="head_null")
    premult=1
    rotationMinMax=(X=4,Y=-26)
    triggerAxis=Z
    twistAxis=(X=1,Y=0,Z=0)
    ExecuteContext=@constrains.ExecutePin
   )
   > RigVMVariableNode(Value=@GetAlphaFromQuatDelta.alpha) {variable=HeadUpRotation}
   > GetAlphaFromQuatDelta_1 = RigVMFunctionReferenceNode(
    Rotation Order=ZYX
    child=(Type=Bone,Name="head")
    debugRotation=false
    outputMinMax=(X=0,Y=1)
    parent=(Type=Null,Name="head_null")
    premult=1
    rotationMinMax=(X=4,Y=12)
    triggerAxis=Z
    twistAxis=(X=1,Y=0,Z=0)
   )
   > RigVMVariableNode(Value=@GetAlphaFromQuatDelta_1.alpha) {variable=HeadDownRotation}
   n1 = RigVMVariableNode() {variable=Is targeting float}
   Multiply = &RigVMFunction_MathDoubleMul(B=@n1.Value, A=RigVMVariableNode() {variable=HeadUpRotation}.Value)
   > &RigUnit_ModifyTransforms(
    ItemToModify="((Item=(Type=Bone,Name=\"armor_collar\"),Transform=(Rotation=(X=0,Y=0,Z=0,W=-1),Translation=(X=-3.4656,Y=-3.7393,Z=-1.7863),Scale3D=(X=1,Y=1,Z=0.8))),(Item=(Type=Bone,Name=\"kin_left_torso_2\"),Transform=(Rotation=(X=0,Y=0,Z=0,W=1),Translation=(X=0,Y=0,Z=0),Sc ...(truncated)"
    WeightMinimum=0
    Weight=@Multiply.Result
   )
   > &RigUnit_ModifyTransforms(
    ItemToModify=((Item=(Type=Bone,Name="kin_front_torso_9"),Transform=(Rotation=(X=0,Y=0,Z=0,W=-1),Translation=(X=0,Y=0,Z=0),Scale3D=(X=0,Y=0,Z=0))))
    WeightMinimum=0
    Weight=&RigVMFunction_MathDoubleMul(A=RigVMVariableNode() {variable=HeadDownRotation}.Value, B=@n1.Value).Result
   )
   > &RigUnit_ModifyTransforms(
    ItemToModify="((Item=(Type=Bone,Name=\"kin_right_torso_4\"),Transform=(Rotation=(X=0,Y=0,Z=0,W=-1),Translation=(X=0,Y=-1.6082,Z=0),Scale3D=(X=0.5,Y=0.5,Z=0.5))),(Item=(Type=Bone,Name=\"kin_left_torso_3\"),Transform=(Rotation=(X=0,Y=0,Z=0,W=1),Translation=(X=0,Y=0,Z=0),Scale ...(truncated)"
    WeightMinimum=0
    Weight=@n1.Value
   )
   > &RigUnit_ModifyTransforms(
    ItemToModify="((Item=(Type=Bone,Name=\"kin_left_torso_1\"),Transform=(Rotation=(X=0,Y=0,Z=0,W=-1),Translation=(X=0,Y=4,Z=0),Scale3D=(X=1,Y=1,Z=1))),(Item=(Type=Bone,Name=\"kin_right_torso_4\"),Transform=(Rotation=(X=0,Y=0,Z=0,W=1),Translation=(X=-1,Y=1,Z=0),Scale3D=(X=1,Y=1 ...(truncated)"
    WeightMinimum=0
    Weight=&RigVMFunction_MathBoolToFloat(Value=RigVMVariableNode() {variable=Is using hang glider}.Value).Result
   )
   setFKCtrl = RigVMFunctionReferenceNode(ExecuteContext=(), setFK_Ctrl=false)
  ]}
 RigVMModel Rig_1 {nodes [
   n1 = &RigUnit_PrepareForExecution()
   addFKCtrl = RigVMFunctionReferenceNode(
    Custom_FKCtrl=()
    Options_FKCtrl="(pelvis_1_#12=true,spine_3_#13=true,clavicle_5_#14=true,arm_7_#15=true,leg_9_#16=true,head_11_46DC15EF465CD2979 ...(truncated)"
    Shape_Settings="(pelvis_2_#17=(bVisible=True,Name=\"Hexagon_Thin\",Color=(R=0.0117,G=0.8281,B=1,A=1),Transform=(Rotation=(X=0,Y=-0.7071,Z=0,W=0.7071),Translation=(X=2,Y=0,Z=0),Scale3D=(X=5,Y=5,Z=7))),spine_4_#18=(bVi ...(truncated)"
    ExecuteContext=@n1.ExecutePin
   )
   > addNullObject = RigVMFunctionReferenceNode(
    Custom_Joints=()
    Options_Null="(pelvis_1_#12=False,spine_3_#13=false,clavicle_5_#14=false,arm_7_#15=true,leg_9_#16=true,head_11_46DC15EF465CD2 ...(truncated)"
   )
  ]}
]
hierarchy [
 root {loc=<0 0 0>, rot=<0 0 0 1>}
 pelvis {parent=root, loc=<0 -0.7071 93.1191>, rot=<0.0127 0.707 -0.0127 -0.707>}
 spine_01 {parent=pelvis, loc=<6.7494 -0.6388 0>, rot=<0 0 0.034 -0.9994>}
 spine_02 {parent=spine_01, loc=<7.2518 -0.035 0>, rot=<0 0 0.0238 -0.9997>}
 spine_03 {parent=spine_02, loc=<6.0104 0 0>, rot=<0 0 0.0149 -0.9999>}
 spine_04 {parent=spine_03, loc=<8.1024 0 0>, rot=<0 0 -0.0132 -0.9999>}
 spine_05 {parent=spine_04, loc=<10.0146 0 0>, rot=<0 0 -0.0015 -1>}
 neck_01 {parent=spine_05, loc=<11.3138 0.1529 0>, rot=<0 0 -0.2003 -0.9797>}
 neck_02 {parent=neck_01, loc=<4.5627 0 0>, rot=<0 0 0.0504 -0.9987>}
 head {parent=neck_02, loc=<5.3914 0 0>, rot=<0 0 0.115 -0.9934>}
 head_null {parent=neck_02, type=Null, loc=<5.3914 0 0>, rot=<0 0 0.115 -0.9934>}
 armor_collar {parent=neck_01, loc=<2.1126 -0.2087 -0.314>, rot=<0 0 -0.6851 -0.7284>}
 attach_cape {parent=spine_05, loc=<9.017 -11.714 0>, rot=<0.0616 -0.7044 -0.0616 -0.7044>}
 cape {parent=attach_cape, type=Null, loc=<0 -3.6014 0>, rot=<0 0 0 1>}
 attach_backpack {parent=spine_05, loc=<-0.123 -13.473 0>, rot=<-0.0093 -0.707 0.0093 -0.707>}
 Backpack_BR {parent=attach_backpack, type=Null, loc=<0 0 0>, rot=<0 0 0 1>}
 clavicle_l {parent=spine_05, loc=<4.3575 -0.1466 -2.3223>, rot=<0.0465 -0.7289 -0.009 -0.6829>}
 upperarm_l {parent=clavicle_l, loc=<15.5111 0 0>, rot=<0.0226 -0.3234 -0.0088 -0.9459>}
 lowerarm_l {parent=upperarm_l, loc=<26.6514 0 0>, rot=<0.0391 -0.0099 -0.2453 -0.9686>}
 hand_l {parent=lowerarm_l, loc=<22.3491 0 0>, rot=<-0.7081 -0.0507 0.0975 -0.6975>}
 thumb_01_l {parent=hand_l, loc=<1.3118 -0.3191 -2.459>, rot=<0.6546 -0.4042 0.0664 -0.6354>}
 thumb_02_l {parent=thumb_01_l, loc=<5.9336 -0.0448 0.0401>, rot=<-0.0037 0.0213 0.298 -0.9543>}
 thumb_03_l {parent=thumb_02_l, loc=<2.2223 0.3682 -0.0907>, rot=<0.0118 -0.013 0.0066 -0.9998>}
 index_metacarpal_l {parent=hand_l, loc=<1.6535 0.0183 -2.96>, rot=<0.0158 -0.0124 -0.0173 -0.9996>}
 index_01_l {parent=index_metacarpal_l, loc=<8.3008 0.0357 -0.0524>, rot=<-0.0044 0.0485 0.0924 -0.9945>}
 index_02_l {parent=index_01_l, loc=<3.3587 -0.3004 0.0452>, rot=<-0.002 0.0027 0.3225 -0.9466>}
 index_03_l {parent=index_02_l, loc=<2.5578 -0.123 -0.0252>, rot=<-0.002 0.0067 0.2558 -0.9667>}
 middle_metacarpal_l {parent=hand_l, loc=<2.1856 -0.1725 -0.5984>, rot=<0.006 0.0168 -0.0329 -0.9993>}
 middle_01_l {parent=middle_metacarpal_l, loc=<7.3275 -0.2447 0.0129>, rot=<-0.0208 0.0042 0.0858 -0.9961>}
 middle_02_l {parent=middle_01_l, loc=<4.0123 0.1401 0.1175>, rot=<-0.0218 0.0258 0.3512 -0.9357>}
 middle_03_l {parent=middle_02_l, loc=<2.6077 0.0534 0.1373>, rot=<0.0082 0.0188 0.1708 -0.9851>}
 ring_metacarpal_l {parent=hand_l, loc=<2.2074 -0.3625 1.3638>, rot=<-0.0284 0.0722 -0.0458 -0.9959>}
 ring_01_l {parent=ring_metacarpal_l, loc=<7.0295 -0.6197 -0.1446>, rot=<-0.0489 -0.0252 0.1892 -0.9804>}
 ring_02_l {parent=ring_01_l, loc=<3.6174 0.4647 -0.1027>, rot=<0 0 0.2806 -0.9598>}
 ring_03_l {parent=ring_02_l, loc=<2.4152 0.114 0.0199>, rot=<0 0 0.1189 -0.9929>}
 pinky_metacarpal_l {parent=hand_l, loc=<1.7616 -0.9753 3.3814>, rot=<-0.0397 0.0814 -0.069 -0.9935>}
 pinky_01_l {parent=pinky_metacarpal_l, loc=<7.1709 -0.4831 -0.1094>, rot=<-0.0654 -0.0035 0.2242 -0.9723>}
 pinky_02_l {parent=pinky_01_l, loc=<2.8274 0.0893 -0.1043>, rot=<0.0037 0.0011 0.2854 -0.9584>}
 pinky_03_l {parent=pinky_02_l, loc=<2.2046 0.0307 -0.0078>, rot=<-0.0002 0.0098 0.1445 -0.9895>}
 deform_wrist_dn_base_l {parent=hand_l, loc=<0 -2.2342 0>, rot=<0 0 0 -1>}
 deform_wrist_dn_l {parent=deform_wrist_dn_base_l, loc=<0 0 0>, rot=<0 0 0 1>}
 deform_wrist_up_base_l {parent=hand_l, loc=<0 2.2682 0>, rot=<0 0 0 -1>}
 deform_wrist_up_l {parent=deform_wrist_up_base_l, loc=<0 0 0>, rot=<0 0 0 1>}
 lowerarm_twist_02_l {parent=lowerarm_l, loc=<9.6199 0 0>, rot=<0 0 0 -1>}
 lowerarm_twist_01_l {parent=lowerarm_l, loc=<17.8549 0 0>, rot=<0 0 0 -1>}
 deform_elbow_in_base_l {parent=lowerarm_l, loc=<0.2112 5.259 0.8274>, rot=<0 0 0 -1>}
 deform_elbow_in_l {parent=deform_elbow_in_base_l, loc=<0 0 0>, rot=<0 0 0 1>}
 deform_elbow_out_base_l {parent=lowerarm_l, loc=<0 -4 0>, rot=<0 0 0 -1>}
 deform_elbow_out_l {parent=deform_elbow_out_base_l, loc=<0 0 0>, rot=<0 0 0 1>}
 upperarm_twist_01_l {parent=upperarm_l, loc=<0.0041 0 0>, rot=<0 0 0 -1>}
 deform_upperarm_bk_base_l {parent=upperarm_twist_01_l, loc=<1.0813 -7.2193 -1.9453>, rot=<0 0 0 1>}
 deform_upperarm_bk_l {parent=deform_upperarm_bk_base_l, loc=<0 0 0>, rot=<0 0 0 1>}
 deform_upperarm_fr_base_l {parent=upperarm_twist_01_l, loc=<-0.3691 7.0506 -2.765>, rot=<0 0 0 1>}
 deform_upperarm_fr_l {parent=deform_upperarm_fr_base_l, loc=<0 0 0>, rot=<0 0 0 1>}
 deform_upperarm_up_base_l {parent=upperarm_twist_01_l, loc=<-3.2216 -1.1391 3.3311>, rot=<0 0 0 1>}
 deform_upperarm_up_l {parent=deform_upperarm_up_base_l, loc=<0 0 0>, rot=<0 0 0 1>}
 upperarm_twist_02_l {parent=upperarm_l, loc=<13.7941 0 0>, rot=<0 0 0 -1>}
 kin_armBoxes_l_1 {parent=upperarm_l, loc=<29.0196 5.9699 4.0641>, rot=<0.6987 0.1307 -0.1391 -0.6895>}
 kin_armBoxes_l_2 {parent=upperarm_l, loc=<31.3085 0.0589 4.1821>, rot=<0.6987 0.1307 -0.1391 -0.6895>}
 kin_armBoxes_l_3 {parent=upperarm_l, loc=<28.7239 8.1917 -2.0374>, rot=<0.6987 0.1307 -0.1391 -0.6895>}
 kin_armBoxes_l_4 {parent=upperarm_l, loc=<32.8508 -1.7285 -2.0244>, rot=<0.6987 0.1307 -0.1391 -0.6895>}
 kin_armBoxes_l_5 {parent=upperarm_l, loc=<28.407 5.4303 -4.7473>, rot=<-0.685 -0.1382 -0.1316 -0.7031>}
 kin_armBoxes_l_6 {parent=upperarm_l, loc=<22.4373 -5.2283 -2.431>, rot=<-0.6975 -0.0126 -0.0092 -0.7164>}
 kin_armBoxes_l_7 {parent=upperarm_l, loc=<31.0623 1.8703 -4.9534>, rot=<-0.685 -0.1382 -0.1316 -0.7031>}
 kin_armBoxes_l_8 {parent=upperarm_l, loc=<20.0211 5.4515 0.4795>, rot=<0.7124 0.0002 -0.0068 -0.7018>}
 kin_armBoxes_l_9 {parent=upperarm_l, loc=<19.6398 6.3679 -2.3163>, rot=<-0.6975 -0.0126 -0.0092 -0.7164>}
 kin_armBoxes_l_10 {parent=upperarm_l, loc=<12.3128 4.8807 -2.4995>, rot=<-0.6975 -0.0126 -0.0092 -0.7164>}
 kin_armBoxes_l_11 {parent=upperarm_l, loc=<7.4455 2.7506 -3.8335>, rot=<-0.6975 -0.0126 -0.0092 -0.7164>}
 kin_armBoxes_l_12 {parent=upperarm_l, loc=<14.4479 -6.3187 -3.6186>, rot=<-0.6975 -0.0126 -0.0092 -0.7164>}
 kin_armBoxes_l_13 {parent=upperarm_l, loc=<20.7042 1.3264 -6.428>, rot=<-0.6975 -0.0126 -0.0092 -0.7164>}
 kin_armBoxes_l_14 {parent=upperarm_l, loc=<22.4586 2.6819 -6.311>, rot=<-0.6975 -0.0126 -0.0092 -0.7164>}
 kin_armBoxes_l_15 {parent=upperarm_l, loc=<21.0493 -2.073 -5.8101>, rot=<-0.704 -0.0235 -0.015 -0.7097>}
 kin_armBoxes_l_16 {parent=upperarm_l, loc=<22.4077 -0.6488 -6.5048>, rot=<-0.704 -0.0235 -0.015 -0.7097>}
 kin_armBoxes_l_17 {parent=upperarm_l, loc=<21.3355 -5.6409 -0.4017>, rot=<0.712 0.0171 -0.0239 -0.7016>}
 clavicle_r {parent=spine_05, loc=<4.3575 -0.1466 2.3223>, rot=<0.7289 0.0465 -0.6829 0.009>}
 upperarm_r {parent=clavicle_r, loc=<-15.5111 0 0>, rot=<0.0226 -0.3234 -0.0088 -0.9459>}
 lowerarm_r {parent=upperarm_r, loc=<-26.6514 0 0>, rot=<0.0391 -0.0099 -0.2453 -0.9686>}
 hand_r {parent=lowerarm_r, loc=<-22.3491 0 0>, rot=<-0.7081 -0.0507 0.0975 -0.6975>}
 thumb_01_r {parent=hand_r, loc=<-1.3118 0.3191 2.459>, rot=<0.6546 -0.4042 0.0664 -0.6354>}
 thumb_02_r {parent=thumb_01_r, loc=<-5.9336 0.0448 -0.0401>, rot=<-0.0037 0.0213 0.298 -0.9543>}
 thumb_03_r {parent=thumb_02_r, loc=<-2.2223 -0.3682 0.0907>, rot=<0.0118 -0.013 0.0066 -0.9998>}
 index_metacarpal_r {parent=hand_r, loc=<-1.6535 -0.0183 2.96>, rot=<0.0158 -0.0124 -0.0173 -0.9996>}
 index_01_r {parent=index_metacarpal_r, loc=<-8.3008 -0.0357 0.0524>, rot=<-0.0044 0.0485 0.0924 -0.9945>}
 index_02_r {parent=index_01_r, loc=<-3.3587 0.3004 -0.0452>, rot=<-0.002 0.0027 0.3225 -0.9466>}
 index_03_r {parent=index_02_r, loc=<-2.5578 0.123 0.0252>, rot=<-0.002 0.0067 0.2558 -0.9667>}
 middle_metacarpal_r {parent=hand_r, loc=<-2.1856 0.1725 0.5984>, rot=<0.006 0.0168 -0.0329 -0.9993>}
 middle_01_r {parent=middle_metacarpal_r, loc=<-7.3275 0.2447 -0.0129>, rot=<-0.0208 0.0042 0.0858 -0.9961>}
 middle_02_r {parent=middle_01_r, loc=<-4.0123 -0.1401 -0.1175>, rot=<-0.0218 0.0258 0.3512 -0.9357>}
 middle_03_r {parent=middle_02_r, loc=<-2.6077 -0.0534 -0.1373>, rot=<0.0082 0.0188 0.1708 -0.9851>}
 ring_metacarpal_r {parent=hand_r, loc=<-2.2074 0.3625 -1.3638>, rot=<-0.0284 0.0722 -0.0458 -0.9959>}
 ring_01_r {parent=ring_metacarpal_r, loc=<-7.0295 0.6197 0.1446>, rot=<-0.0489 -0.0252 0.1892 -0.9804>}
 ring_02_r {parent=ring_01_r, loc=<-3.6174 -0.4647 0.1027>, rot=<0 0 0.2806 -0.9598>}
 ring_03_r {parent=ring_02_r, loc=<-2.4152 -0.114 -0.0199>, rot=<0 0 0.1189 -0.9929>}
 pinky_metacarpal_r {parent=hand_r, loc=<-1.7616 0.9753 -3.3814>, rot=<-0.0397 0.0814 -0.069 -0.9935>}
 pinky_01_r {parent=pinky_metacarpal_r, loc=<-7.1709 0.4831 0.1094>, rot=<-0.0654 -0.0035 0.2242 -0.9723>}
 pinky_02_r {parent=pinky_01_r, loc=<-2.8274 -0.0893 0.1043>, rot=<0.0037 0.0011 0.2854 -0.9584>}
 pinky_03_r {parent=pinky_02_r, loc=<-2.2046 -0.0307 0.0078>, rot=<-0.0002 0.0098 0.1445 -0.9895>}
 deform_wrist_dn_base_r {parent=hand_r, loc=<-0.0004 2.2345 -0.0001>, rot=<0 0 0 -1>}
 deform_wrist_dn_r {parent=deform_wrist_dn_base_r, loc=<0 0 0>, rot=<0 0 0 1>}
 deform_wrist_up_base_r {parent=hand_r, loc=<0.0003 -2.2684 0.0001>, rot=<0 0 0 -1>}
 deform_wrist_up_r {parent=deform_wrist_up_base_r, loc=<0 0 0>, rot=<0 0 0 1>}
 lowerarm_twist_02_r {parent=lowerarm_r, loc=<-9.6199 0 0>, rot=<0 0 0 -1>}
 lowerarm_twist_01_r {parent=lowerarm_r, loc=<-17.8549 0 0>, rot=<0 0 0 -1>}
 deform_elbow_in_base_r {parent=lowerarm_r, loc=<-0.2111 -5.2591 -0.8275>, rot=<0 0 0 -1>}
 deform_elbow_in_r {parent=deform_elbow_in_base_r, loc=<0 0 0>, rot=<0 0 0 1>}
 deform_elbow_out_base_r {parent=lowerarm_r, loc=<0 4 -0.0001>, rot=<0 0 0 -1>}
 deform_elbow_out_r {parent=deform_elbow_out_base_r, loc=<0 0 0>, rot=<0 0 0 1>}
 upperarm_twist_01_r {parent=upperarm_r, loc=<-0.0041 0 0>, rot=<0 0 0 -1>}
 deform_upperarm_bk_base_r {parent=upperarm_twist_01_r, loc=<-1.0814 7.2192 1.9455>, rot=<0 0 0 1>}
 deform_upperarm_bk_r {parent=deform_upperarm_bk_base_r, loc=<0 0 0>, rot=<0 0 0 1>}
 deform_upperarm_fr_base_r {parent=upperarm_twist_01_r, loc=<0.3689 -7.0506 2.7652>, rot=<0 0 0 1>}
 deform_upperarm_fr_r {parent=deform_upperarm_fr_base_r, loc=<0 0 0>, rot=<0 0 0 1>}
 deform_upperarm_up_base_r {parent=upperarm_twist_01_r, loc=<3.2214 1.1391 -3.3308>, rot=<0 0 0 1>}
 deform_upperarm_up_r {parent=deform_upperarm_up_base_r, loc=<0 0 0>, rot=<0 0 0 -1>}
 upperarm_twist_02_r {parent=upperarm_r, loc=<-13.7941 0 0>, rot=<0 0 0 -1>}
 kin_armBoxes_r_1 {parent=upperarm_r, loc=<-29.6783 -8.0482 2.4501>, rot=<-0.1736 0.6767 0.704 -0.1279>}
 kin_armBoxes_r_2 {parent=upperarm_r, loc=<-28.407 -5.4303 4.7473>, rot=<0.1224 -0.6824 0.6984 -0.1775>}
 kin_armBoxes_r_3 {parent=upperarm_r, loc=<-28.8338 -2.8332 6.0109>, rot=<0.1224 -0.6824 0.6984 -0.1775>}
 kin_armBoxes_r_4 {parent=upperarm_r, loc=<-20.7042 -1.3264 6.4279>, rot=<-0.0006 -0.7029 0.7091 -0.0565>}
 kin_armBoxes_r_5 {parent=upperarm_r, loc=<-33.8897 1.7094 -0.5205>, rot=<-0.1736 0.6767 0.704 -0.1279>}
 kin_armBoxes_r_6 {parent=upperarm_r, loc=<-20.0149 -5.7625 -0.4952>, rot=<-0.0384 0.6887 0.724 0.0014>}
 kin_armBoxes_r_7 {parent=upperarm_r, loc=<-19.8705 -9.5245 -2.973>, rot=<0.9345 0.014 0.3555 -0.0134>}
 kin_armBoxes_r_8 {parent=upperarm_r, loc=<-12.3128 -4.8807 2.4995>, rot=<0.0151 -0.7041 0.7099 -0.0067>}
 kin_armBoxes_r_9 {parent=upperarm_r, loc=<-7.4455 -2.7506 3.8335>, rot=<0.0151 -0.7041 0.7099 -0.0067>}
 kin_armBoxes_r_10 {parent=upperarm_r, loc=<-3.6504 -1.824 -1.4842>, rot=<0.9345 0.014 0.3555 -0.0134>}
 kin_armBoxes_r_11 {parent=upperarm_r, loc=<-31.053 -1.8703 5.1239>, rot=<0.1224 -0.6824 0.6984 -0.1775>}
 kin_armBoxes_r_12 {parent=upperarm_r, loc=<-14.4477 6.3188 3.6183>, rot=<0.0151 -0.7041 0.7099 -0.0067>}
 kin_armBoxes_r_13 {parent=upperarm_r, loc=<-20.7045 -1.3264 6.4283>, rot=<0.0326 -0.7037 0.7093 -0.0241>}
 kin_armBoxes_r_14 {parent=upperarm_r, loc=<-22.3425 -4.8049 6.3271>, rot=<0.0235 -0.704 0.7097 -0.015>}
 kin_armBoxes_r_15 {parent=upperarm_r, loc=<-22.5583 0.6375 6.4358>, rot=<0.0151 -0.7041 0.7099 -0.0067>}
 kin_armBoxes_r_16 {parent=upperarm_r, loc=<-21.219 5.2092 3.1136>, rot=<0.0151 -0.7041 0.7099 -0.0067>}
 kin_armBoxes_r_17 {parent=upperarm_r, loc=<-21.9899 5.6153 0.4078>, rot=<0.0105 -0.7122 -0.7017 0.0172>}
 kin_torso_1 {parent=spine_05, loc=<3.4644 4.3398 -13.3583>, rot=<0.0284 -0.7065 -0.0284 -0.7065>}
 kin_torso_2 {parent=spine_05, loc=<7.245 0.5784 -13.3911>, rot=<0.0284 -0.7065 -0.0284 -0.7065>}
 kin_torso_3 {parent=spine_05, loc=<8.6144 -5.6597 -14.8919>, rot=<0.0284 -0.7065 -0.0284 -0.7065>}
 kin_torso_4 {parent=spine_05, loc=<10.4324 -3.2254 -12.855>, rot=<0.0284 -0.7065 -0.0284 -0.7065>}
 kin_torso_5 {parent=spine_05, loc=<-3.4627 6.3734 -12.1891>, rot=<0.0284 -0.7065 -0.0284 -0.7065>}
 kin_torso_6 {parent=spine_05, loc=<1.4749 9.1891 -7.1419>, rot=<0.0284 -0.7065 -0.0284 -0.7065>}
 kin_torso_7 {parent=spine_05, loc=<6.3545 2.805 -8.4896>, rot=<0.0284 -0.7065 -0.0284 -0.7065>}
 kin_torso_8 {parent=spine_05, loc=<-6.0603 8.2351 -2.5218>, rot=<0.0284 -0.7065 -0.0284 -0.7065>}
 kin_torso_9 {parent=spine_05, loc=<-10.4108 11.7131 -7.4587>, rot=<0.0284 -0.7065 -0.0284 -0.7065>}
 kin_torso_10 {parent=spine_05, loc=<-9.2753 12.1525 2.62>, rot=<0.0284 -0.7065 -0.0284 -0.7065>}
 kin_torso_11 {parent=spine_05, loc=<-1.7308 12.7066 4.1212>, rot=<0.0284 -0.7065 -0.0284 -0.7065>}
 kin_torso_12 {parent=spine_05, loc=<-5.0693 8.6563 6.1585>, rot=<0.0284 -0.7065 -0.0284 -0.7065>}
 kin_torso_13 {parent=spine_05, loc=<-8.7083 13.6326 9.0058>, rot=<0.0284 -0.7065 -0.0284 -0.7065>}
 kin_torso_14 {parent=spine_05, loc=<-9.559 10.5433 8.0106>, rot=<0.0284 -0.7065 -0.0284 -0.7065>}
 kin_torso_15 {parent=spine_05, loc=<2.0559 12.033 0.403>, rot=<0.0284 -0.7065 -0.0284 -0.7065>}
 kin_torso_16 {parent=spine_05, loc=<3.3785 9.2374 5.9609>, rot=<0.0284 -0.7065 -0.0284 -0.7065>}
 kin_torso_17 {parent=spine_05, loc=<2.9495 5.8417 12.5574>, rot=<0.0284 -0.7065 -0.0284 -0.7065>}
 kin_torso_18 {parent=spine_05, loc=<-3.4627 6.3734 12.2101>, rot=<0.0284 -0.7065 -0.0284 -0.7065>}
 kin_torso_19 {parent=spine_05, loc=<7.0253 0.0802 14.8377>, rot=<0.0284 -0.7065 -0.0284 -0.7065>}
 kin_torso_20 {parent=spine_05, loc=<9.5521 -3.6188 10.9411>, rot=<0.0284 -0.7065 -0.0284 -0.7065>}
 kin_torso_21 {parent=spine_05, loc=<11.0485 -9.6686 -3.9381>, rot=<0.0284 -0.7065 -0.0284 -0.7065>}
 kin_torso_22 {parent=spine_05, loc=<4.6767 -6.6752 -11.2207>, rot=<0.0284 -0.7065 -0.0284 -0.7065>}
 kin_torso_23 {parent=spine_05, loc=<2.4386 -8.8242 -15.7578>, rot=<0.0284 -0.7065 -0.0284 -0.7065>}
 kin_torso_24 {parent=spine_05, loc=<-1.3572 -10.7326 -7.9667>, rot=<0.0284 -0.7065 -0.0284 -0.7065>}
 kin_torso_25 {parent=spine_05, loc=<2.4875 -11.6621 -4.4553>, rot=<0.0284 -0.7065 -0.0284 -0.7065>}
 kin_torso_26 {parent=spine_05, loc=<5.7881 -9.1584 1.9262>, rot=<0.0284 -0.7065 -0.0284 -0.7065>}
 kin_torso_27 {parent=spine_05, loc=<7.8907 -11.0988 5.1123>, rot=<0.0284 -0.7065 -0.0284 -0.7065>}
 kin_torso_28 {parent=spine_05, loc=<4.6767 -6.6752 11.2418>, rot=<0.0284 -0.7065 -0.0284 -0.7065>}
 kin_torso_29 {parent=spine_05, loc=<2.9646 -7.824 14.7858>, rot=<0.0284 -0.7065 -0.0284 -0.7065>}
 kin_torso_30 {parent=spine_05, loc=<0.9711 -7.4524 9.5776>, rot=<0.0284 -0.7065 -0.0284 -0.7065>}
 kin_torso_31 {parent=spine_05, loc=<-3.7776 -6.2439 13.7827>, rot=<0.0284 -0.7065 -0.0284 -0.7065>}
 kin_torso_32 {parent=spine_05, loc=<-10.847 -4.5973 9.3152>, rot=<0.0284 -0.7065 -0.0284 -0.7065>}
 kin_torso_33 {parent=spine_05, loc=<-5.258 -4.8338 9.5776>, rot=<0.0284 -0.7065 -0.0284 -0.7065>}
 kin_torso_34 {parent=spine_05, loc=<-8.1753 -9.4497 2.5671>, rot=<0.0284 -0.7065 -0.0284 -0.7065>}
 kin_torso_35 {parent=spine_05, loc=<-14.6744 -7.189 5.5356>, rot=<0.0284 -0.7065 -0.0284 -0.7065>}
 kin_torso_36 {parent=spine_05, loc=<-16.3546 -10.1073 0.3447>, rot=<0.0284 -0.7065 -0.0284 -0.7065>}
 kin_torso_37 {parent=spine_05, loc=<-11.6739 -8.9327 -5.1986>, rot=<0.0284 -0.7065 -0.0284 -0.7065>}
 kin_torso_38 {parent=spine_05, loc=<-9.1258 -5.9677 -6.8231>, rot=<0.0284 -0.7065 -0.0284 -0.7065>}
 kin_torso_39 {parent=spine_05, loc=<-4.0532 -9.5689 -2.9869>, rot=<0.0284 -0.7065 -0.0284 -0.7065>}
 kin_torso_40 {parent=spine_05, loc=<-1.5764 -8.3413 4.1107>, rot=<0.0284 -0.7065 -0.0284 -0.7065>}
 kin_torso_41 {parent=spine_05, loc=<-3.7776 -6.2439 -14.6993>, rot=<0.0284 -0.7065 -0.0284 -0.7065>}
 kin_torso_42 {parent=spine_05, loc=<15.5653 1.7329 -8.0873>, rot=<0.0284 -0.7065 -0.0284 -0.7065>}
 kin_torso_43 {parent=spine_05, loc=<17.7256 -1.6789 -3.9278>, rot=<0.0284 -0.7065 -0.0284 -0.7065>}
 kin_torso_44 {parent=spine_05, loc=<15.9875 3.8875 3.1294>, rot=<0.0284 -0.7065 -0.0284 -0.7065>}
 kin_torso_45 {parent=spine_05, loc=<12.2794 6.1591 -4.7011>, rot=<0.0284 -0.7065 -0.0284 -0.7065>}
 kin_torso_46 {parent=spine_05, loc=<10.4335 7.4001 -1.0632>, rot=<0.0284 -0.7065 -0.0284 -0.7065>}
 kin_torso_47 {parent=spine_05, loc=<12.0902 2.7678 4.166>, rot=<0.0284 -0.7065 -0.0284 -0.7065>}
 kin_torso_48 {parent=spine_05, loc=<-1.6619 11.7851 -0.9643>, rot=<0.0284 -0.7065 -0.0284 -0.7065>}
 kin_left_torso_1 {parent=spine_05, loc=<4.221 4.5262 -16.5628>, rot=<0.0284 -0.7065 -0.0284 -0.7065>}
 kin_left_torso_2 {parent=spine_05, loc=<7.0375 -0.7641 -17.2407>, rot=<0.0284 -0.7065 -0.0284 -0.7065>}
 kin_left_torso_3 {parent=spine_05, loc=<7.9494 1.6611 -11.8387>, rot=<0.0284 -0.7065 -0.0284 -0.7065>}
 kin_left_torso_4 {parent=spine_05, loc=<5.6144 -5.9425 -14.5429>, rot=<0.0284 -0.7065 -0.0284 -0.7065>}
 kin_front_torso_1 {parent=spine_05, loc=<0.2067 10.8246 -9.3992>, rot=<0.0284 -0.7065 -0.0284 -0.7065>}
 kin_front_torso_2 {parent=spine_05, loc=<-9.9429 11.9738 -7.889>, rot=<0.0284 -0.7065 -0.0284 -0.7065>}
 kin_front_torso_3 {parent=spine_05, loc=<-19.9189 12.8911 -2.4309>, rot=<0.0284 -0.7065 -0.0284 -0.7065>}
 kin_front_torso_4 {parent=spine_05, loc=<0.7964 11.8077 2.6686>, rot=<0.0284 -0.7065 -0.0284 -0.7065>}
 kin_front_torso_5 {parent=spine_05, loc=<-6.7125 12.3595 1.5005>, rot=<0.0284 -0.7065 -0.0284 -0.7065>}
 kin_front_torso_6 {parent=spine_05, loc=<-13.1283 9.6074 3.0066>, rot=<0.0284 -0.7065 -0.0284 -0.7065>}
 kin_front_torso_7 {parent=spine_05, loc=<-8.9571 11.0622 9.1979>, rot=<0.0284 -0.7065 -0.0284 -0.7065>}
 kin_front_torso_8 {parent=spine_05, loc=<8.8273 6.138 1.7218>, rot=<0.0284 -0.7065 -0.0284 -0.7065>}
 kin_front_torso_9 {parent=spine_05, loc=<-2.6379 10.8492 13.4971>, rot=<0.0284 -0.7065 -0.0284 -0.7065>}
 kin_back_torso_1 {parent=spine_05, loc=<-1.8667 -10.9739 -9.4303>, rot=<0.0284 -0.7065 -0.0284 -0.7065>}
 kin_back_torso_2 {parent=spine_05, loc=<3.4431 -10.3346 -4.4072>, rot=<0.0284 -0.7065 -0.0284 -0.7065>}
 kin_back_torso_3 {parent=spine_05, loc=<11.397 -7.7287 3.8288>, rot=<0.0284 -0.7065 -0.0284 -0.7065>}
 kin_back_torso_4 {parent=spine_05, loc=<6.8354 -8.8511 3.1151>, rot=<0.0284 -0.7065 -0.0284 -0.7065>}
 kin_back_torso_5 {parent=spine_05, loc=<-10.7924 -9.6488 -6.5317>, rot=<0.0284 -0.7065 -0.0284 -0.7065>}
 kin_back_torso_6 {parent=spine_05, loc=<-8.9133 -8.8476 1.5797>, rot=<0.0284 -0.7065 -0.0284 -0.7065>}
 kin_back_torso_7 {parent=spine_05, loc=<-15.4685 -9.084 2.4241>, rot=<0.0284 -0.7065 -0.0284 -0.7065>}
 kin_back_torso_8 {parent=spine_05, loc=<-20.3892 -9.0316 -0.9052>, rot=<0.0284 -0.7065 -0.0284 -0.7065>}
 kin_back_torso_9 {parent=spine_05, loc=<1.9192 -11.4161 12.5341>, rot=<0.0284 -0.7065 -0.0284 -0.7065>}
 kin_back_torso_10 {parent=spine_05, loc=<16.4735 -4.2635 -4.2798>, rot=<0.0284 -0.7065 -0.0284 -0.7065>}
 kin_right_torso_1 {parent=spine_05, loc=<7.7372 -3.016 9.4323>, rot=<0.0284 -0.7065 -0.0284 -0.7065>}
 kin_right_torso_2 {parent=spine_05, loc=<7.0375 -0.7641 17.2407>, rot=<0.0284 -0.7065 -0.0284 -0.7065>}
 kin_right_torso_3 {parent=spine_05, loc=<2.6124 2.6048 17.4951>, rot=<0.0284 -0.7065 -0.0284 -0.7065>}
 kin_right_torso_4 {parent=spine_05, loc=<4.3435 7.6162 12.5448>, rot=<0.0284 -0.7065 -0.0284 -0.7065>}
 deform_pec_aim_base_l {parent=spine_04, loc=<7.8732 12.6388 -2.4123>, rot=<0.0102 -0.4602 0.3979 -0.7936>}
 deform_pec_aim_l {parent=deform_pec_aim_base_l, loc=<0 0 0>, rot=<0 0 0 1>}
 deform_pec_base_l {parent=deform_pec_aim_l, loc=<10.766 0 0>, rot=<0 0 0 1>}
 deform_pec_l {parent=deform_pec_base_l, loc=<0.055 0 0>, rot=<0 0 0 1>}
 deform_pec_aim_base_r {parent=spine_04, loc=<7.8727 12.6388 2.4123>, rot=<-0.4602 -0.0102 0.7936 0.3979>}
 deform_pec_aim_r {parent=deform_pec_aim_base_r, loc=<0 0 0>, rot=<0 0 0 1>}
 deform_pec_base_r {parent=deform_pec_aim_r, loc=<-10.7662 -0.0006 -0.0007>, rot=<0 0 0 1>}
 deform_pec_r {parent=deform_pec_base_r, loc=<0 0 0>, rot=<0 0 0 1>}
 thigh_l {parent=pelvis, loc=<-2.5693 -0.0109 -7.7501>, rot=<0.0012 0.0572 0.0155 -0.9982>}
 calf_l {parent=thigh_l, loc=<-39.0484 0 0>, rot=<0.0006 -0.0161 -0.0391 -0.9991>}
 foot_l {parent=calf_l, loc=<-42.561 0.1077 0.4292>, rot=<0.0418 -0.0455 0.0343 -0.9975>}
 ball_l {parent=foot_l, loc=<-6.1599 13.8264 0.6806>, rot=<0.0001 -0.0001 -0.7063 -0.7079>}
 calf_twist_02_l {parent=calf_l, loc=<-18.8064 -0.0967 0.0397>, rot=<0.0237 0.0095 -0.0035 -0.9997>}
 calf_twist_01_l {parent=calf_l, loc=<-35.813 -0.0967 0.0397>, rot=<0.0237 0.0095 -0.0035 -0.9997>}
 deform_knee_in_base_l {parent=calf_l, loc=<1.298 -5.2957 -0.0496>, rot=<0 0 0 -1>}
 deform_knee_in_l {parent=deform_knee_in_base_l, loc=<0 0 0>, rot=<0 0 0 1>}
 deform_knee_out_base_l {parent=calf_l, loc=<0 4 0>, rot=<0 0 0 -1>}
 deform_knee_out_l {parent=deform_knee_out_base_l, loc=<0 0 0>, rot=<0 0 0 1>}
 thigh_twist_01_l {parent=thigh_l, loc=<-21.6303 0 0>, rot=<0.0013 0 0 -1>}
 deform_groin_base_l {parent=thigh_l, loc=<-2.1908 8.1142 -1.7442>, rot=<0 0 0 -1>}
 deform_groin_l {parent=deform_groin_base_l, loc=<0 0 0>, rot=<0 0 0 1>}
 thigh_r {parent=pelvis, loc=<-2.5693 -0.0109 7.7501>, rot=<0.0572 -0.0012 0.9982 0.0155>}
 calf_r {parent=thigh_r, loc=<39.0484 0 0>, rot=<0.0006 -0.0161 -0.0391 -0.9991>}
 foot_r {parent=calf_r, loc=<42.561 -0.1077 -0.4292>, rot=<0.0418 -0.0455 0.0343 -0.9975>}
 ball_r {parent=foot_r, loc=<6.1599 -13.8264 -0.6806>, rot=<0.0001 -0.0001 -0.7063 -0.7079>}
 calf_twist_02_r {parent=calf_r, loc=<18.8064 0.0967 -0.0397>, rot=<0.0237 0.0095 -0.0035 -0.9997>}
 calf_twist_01_r {parent=calf_r, loc=<35.813 0.0967 -0.0397>, rot=<0.0237 0.0095 -0.0035 -0.9997>}
 deform_knee_in_base_r {parent=calf_r, loc=<-1.298 5.2957 0.0496>, rot=<0 0 0 -1>}
 deform_knee_in_r {parent=deform_knee_in_base_r, loc=<0 0 0>, rot=<0 0 0 1>}
 deform_knee_out_base_r {parent=calf_r, loc=<0 -4 0>, rot=<0 0 0 -1>}
 deform_knee_out_r {parent=deform_knee_out_base_r, loc=<0 0 0>, rot=<0 0 0 1>}
 thigh_twist_01_r {parent=thigh_r, loc=<21.6303 0 0>, rot=<0.0013 0 0 -1>}
 deform_groin_base_r {parent=thigh_r, loc=<2.1908 -8.1142 1.7442>, rot=<0 0 0 -1>}
 deform_groin_r {parent=deform_groin_base_r, loc=<0 0 0>, rot=<0 0 0 1>}
 deform_glute_base_l {parent=pelvis, loc=<-8.8064 -8.8222 -8.4287>, rot=<0.0275 0.0501 -0.4472 -0.8926>}
 deform_glute_l {parent=deform_glute_base_l, loc=<0 0 0>, rot=<0 0 0 1>}
 deform_glute_base_r {parent=pelvis, loc=<-8.8062 -8.8222 8.4287>, rot=<0.0501 -0.0275 0.8926 -0.4472>}
 deform_glute_r {parent=deform_glute_base_r, loc=<0 0 0>, rot=<0 0 0 1>}
]
variables [
 HeadDownRotation {type=double, default=0}
 HeadUpRotation {type=double, default=0}
 Is targeting float {type=double, default=0}
 Is using hang glider {type=bool, default=False}
]
