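# Post-process style animation Blueprint for the SpacePlunge Hunt male body, layering a Control Rig onto the pose copied from the attached parent mesh.
# Each update eases a targeting float and derives a spring-controller "cubes" alpha that fades with vertical speed, peaking lower during skydive dive or float.
# Parachute open and close events maintain a hang-glider flag.
# The rig exposes targeting and hang-glider inputs, though no graph wires link these variables to it.
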AnimBlueprint /BRCosmetics/Characters/Player/Male/Medium/Bodies/M_MED_SpacePlunge_Hunt/Meshes/M_MED_SpacePlunge_Hunt_AnimBP
target_skeleton=&/M_MED_SpacePlunge_Hunt_Skeleton
graphs [
 AnimGraph {nodes [
   n1 = Control Rig!(Source=Copy Pose From Mesh() {bUseAttachedParent=true}, Alpha, Is targeting float:double=0, Is using hang glider:bool=false) -> (Pose) {controlRigAssetReference {blueprintRigClass=&/M_MED_SpacePlunge_Hunt_CtrlRig.M_MED_SpacePlunge_Hunt_CtrlRig_C}}
   Output Pose(Result=@n1) {layerGroup=None}
  ]}
 EventGraph {nodes [
   n1 = Event BlueprintUpdateAnimation()
   > Sequence()
   n2 = Get bIsTargeting()
   > Set isTargeting float(isTargeting float:double=FInterpTo(
     Current=Get isTargeting float() -> (isTargeting float:double)
     Target="To Float (Boolean)"(InBool=@n2)
     DeltaTime=@n1.DeltaTimeX
     InterpSpeed=15
    )) -> (Output_Get:double)
   > n3 = Branch(Condition=OR Boolean(A=Get bIsSkydiveDiveMode(), B=Get bIsSkydiveFloating()))
   n4 = Get PawnVelocityZ()
   > Set springControllerCubes alpha(springControllerCubes alpha:double=Remap Float(Input="Absolute (Float)"(A=@n4), OldRange=(X=900,Y=1250), NewRange=(X=0.02,Y=0), Clamp=true) -> (Output=~)) -> (Output_Get:double)
   n5 = Get PawnVelocityZ()
   Set springControllerCubes alpha(
    execute=@n3.else
    springControllerCubes alpha=Remap Float(Input="Absolute (Float)"(A=@n5), OldRange=(X=900,Y=1250), NewRange=(X=0.05,Y=0), Clamp=true) -> (Output=~)
   )
   Event OnParachuteOpened()
   > Switch on EGliderType(Selection=Get PlayerGliderType() -> (PlayerGliderType=HangGlider))
   > Set isUsingHangGlidingInParachute(isUsingHangGlidingInParachute:bool=true) -> (Output_Get:bool)
   Event OnParachuteClosed()
   > Set isUsingHangGlidingInParachute(isUsingHangGlidingInParachute=false)
  ]}
]
variables [
 isTargeting float {type=number}
 isUsingHangGlidingInParachute {type=boolean}
 springControllerCubes alpha {type=number}
]
